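# Player item animation layer Blueprint for the ElectroBat melee weapon, overriding full-body, upper-body and landing poses.
# Full-body input passes through a charged tac-sprint transition machine, then a charged-swing hit/miss outro machine, then a pelvis yaw twist from CustomMeleeTwist.
# LowerBody blend masks keep legs on locomotion during the transition and moving outros.
# Upper body swaps to a sliding pose; landing layers an additive from the right clavicle.
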AnimBlueprint /ElectroBatGameplay/Assets/Animation/ElectroBat_Player_AnimBP
target_skeleton=&Fortnite_M_Avg_Player_Skeleton
graphs [
 AnimGraph {nodes [Output Pose()]}
 EventGraph {links [], nodes []}
 ItemAimOffset {nodes [
   Input Pose!() -> (Pose, AimOffsetAlpha:double, Yaw:double, Pitch:double) {name=InPose_AimOffset}
   Input Pose!() -> (Pose) {name=InPose_PreAimOffsets}
   Output Pose() {layerGroup=Item Override}
  ]}
 ItemFallAdditive {nodes [
   Input Pose() {name=InPoseItemFallAdditive}
   Output Pose() {layerGroup=Item Override}
  ]}
 ItemFallLandAdditive {nodes [
   Output Pose(Result=Layered blend per bone(
     BasePose=Use cached pose 'ItemFallLand'()
     BlendPoses_0="Sequence Player (additive) 'ElectroBat_JumpLand_Additive_Pose'"() {bLoopAnimation=false, sequence=&/ElectroBat_JumpLand_Additive_Pose}
     BlendWeights_0=0.75
    ) {layerSetup [{branchFilters [{boneName=clavicle_r}]}]}) {layerGroup=Item Override}
   Save cached pose 'ItemFallLand'(Pose=Input Pose() {name=InPoseItemFallLandAdditive}) {cacheName=ItemFallLand}
  ]}
 ItemFallLevel2Override {nodes [
   Input Pose() {name=InPose_FallLevel2}
   Output Pose() {layerGroup=Item Override}
  ]}
 ItemFinalPoseOverride {nodes [
   Input Pose() {name=InFinalPose}
   Output Pose() {layerGroup=Item Override}
  ]}
 ItemFullBodyAimOffsets {nodes [
   Input Pose!() -> (Pose, FullBody_AimOffsetAlpha:double, Aim_Yaw:double, Aim_Pitch:double) {name=InPoseFullBodyAimOffsets}
   Input Pose!() -> (Pose) {name=InPosePreFullBodyAimOffsets}
   Output Pose() {layerGroup=Item Override}
  ]}
 ItemFullBodyOverride {nodes [
   n1 = "Transform (Modify) Bone - Bone: pelvis"!(
    ComponentPose=Local To Component(LocalPose=Use cached pose 'AfterMontageCache'())
    Alpha=1
    Rotation_Roll:float=0
    Rotation_Pitch:float=0
    Rotation_Yaw:float=Get CustomMeleeTwist() -> (CustomMeleeTwist:float)
   ) -> (Pose) {
    rotationMode=BMM_Additive
    rotationSpace=BCS_WorldSpace
    translationSpace=BCS_WorldSpace
    alphaBoolBlend {blendOption=Linear}
    boneToModify {boneName=pelvis}
   }
   Output Pose(Result=Component To Local(ComponentPose=@n1)) {layerGroup=Item Override}
   AnimGraphNode_StateMachine_0 = MontageOutro() {is=title AnimGraphNode_StateMachine}
   Save cached pose 'AfterMontageCache'(Pose=@AnimGraphNode_StateMachine_0) {cacheName=AfterMontageCache}
   AnimGraphNode_StateMachine_1 = IntoChargeState() {is=title AnimGraphNode_StateMachine}
   Save cached pose 'IntoChargeStateCache'(Pose=@AnimGraphNode_StateMachine_1) {cacheName=IntoChargeStateCache}
   Save cached pose 'FullBodyCache'(Pose=Input Pose() {name=InPoseFullBody}) {cacheName=FullBodyCache}
  ]}
 MontageOutro {
  entry=State
  path=ItemFullBodyOverride.AnimGraphNode_StateMachine_0.MontageOutro
  ids=id
  nodes [
   AnimStateConduitNode_1 = Conduit()
   AnimStateNode_0 = State() {is=title AnimStateNode}
   AnimStateNode_1 = Miss() {is=title AnimStateNode, stateLeft {guid=#1}}
   AnimStateNode_2 = Hit() {is=title AnimStateNode, stateLeft {guid=#2}}
   AnimStateTransitionNode_1 = AnimStateTransitionNode()
   AnimStateTransitionNode_5 = AnimStateTransitionNode()
   AnimStateTransitionNode_6 = AnimStateTransitionNode() {crossfadeDuration=0.1}
   AnimStateTransitionNode_7 = AnimStateTransitionNode() {crossfadeDuration=0.1}
   AnimStateTransitionNode_8 = AnimStateTransitionNode()
  ]
  links [
   {
    dst=Hit
    rule=ItemFullBodyOverride.AnimGraphNode_StateMachine_0.MontageOutro.AnimStateTransitionNode_6.Transition
    src=Conduit
    via=AnimStateTransitionNode_6
   }
   {
    dst=Miss
    rule=ItemFullBodyOverride.AnimGraphNode_StateMachine_0.MontageOutro.AnimStateTransitionNode_7.Transition
    src=Conduit
    via=AnimStateTransitionNode_7
   }
   {
    dst=State
    rule=ItemFullBodyOverride.AnimGraphNode_StateMachine_0.MontageOutro.AnimStateTransitionNode_8.Transition
    src=Hit
    via=AnimStateTransitionNode_8
   }
   {
    dst=State
    rule=ItemFullBodyOverride.AnimGraphNode_StateMachine_0.MontageOutro.AnimStateTransitionNode_8.Transition
    src=Miss
    via=AnimStateTransitionNode_1
   }
   {
    dst=Conduit
    rule=ItemFullBodyOverride.AnimGraphNode_StateMachine_0.MontageOutro.AnimStateTransitionNode_5.Transition
    src=State
    via=AnimStateTransitionNode_5
   }
  ]
 }
 Conduit {
  path=ItemFullBodyOverride.AnimGraphNode_StateMachine_0.MontageOutro.AnimStateConduitNode_1.Conduit
  nodes [Result(bCanEnterTransition=Get bUserCurve01() -> (bUserCurve01:bool))]
 }
 State {
  path=ItemFullBodyOverride.AnimGraphNode_StateMachine_0.MontageOutro.AnimStateNode_0.State
  nodes [Output Animation Pose(Result=Use cached pose 'IntoChargeStateCache'())]
 }
 Miss {
  path=ItemFullBodyOverride.AnimGraphNode_StateMachine_0.MontageOutro.AnimStateNode_1.Miss
  nodes [
   n1 = Sequence Player 'ElectroBat_ChargeUp_Outro_Miss'() {bLoopAnimation=false, groupName=Outro, method=SyncGroup, sequence=&/ElectroBat_ChargeUp_Outro_Miss}
   n2 = Layered blend per bone(BasePose=@n1, BlendPoses_0=Use cached pose 'IntoChargeStateCache'()) {
    bMeshSpaceRotationBlend=true
    blendMode=BlendMask
    blendMasks [&Fortnite_M_Avg_Player_Skeleton:LowerBody]
    layerSetup []
   }
   n3 = Sequence Player 'ElectroBat_ChargeUp_Outro_Miss'() {bLoopAnimation=false, groupName=Outro, method=SyncGroup, sequence=&ElectroBat_ChargeUp_Outro_Miss}
   Output Animation Pose(Result=Blend Poses by bool(BlendPose_0=@n2, BlendPose_1=@n3, BlendTime_0=0.2, BlendTime_1=0.2, bActiveValue=Get bIsMoving2D()) {blendType=Linear})
  ]
 }
 Hit {
  path=ItemFullBodyOverride.AnimGraphNode_StateMachine_0.MontageOutro.AnimStateNode_2.Hit
  nodes [
   n1 = Sequence Player 'ElectroBat_ChargeUp_Outro_Hit'() {bLoopAnimation=false, groupName=Outro, method=SyncGroup, sequence=&/ElectroBat_ChargeUp_Outro_Hit}
   n2 = Layered blend per bone(BasePose=@n1, BlendPoses_0=Use cached pose 'IntoChargeStateCache'()) {
    bMeshSpaceRotationBlend=true
    blendMode=BlendMask
    blendMasks [&Fortnite_M_Avg_Player_Skeleton:LowerBody]
    layerSetup []
   }
   n3 = Sequence Player 'ElectroBat_ChargeUp_Outro_Hit'() {bLoopAnimation=false, groupName=Outro, method=SyncGroup, sequence=&ElectroBat_ChargeUp_Outro_Hit}
   Output Animation Pose(Result=Blend Poses by bool(BlendPose_0=@n2, BlendPose_1=@n3, BlendTime_0=0.2, BlendTime_1=0.2, bActiveValue=Get bIsMoving2D()) {blendType=Linear})
  ]
 }
 Transition {
  path=ItemFullBodyOverride.AnimGraphNode_StateMachine_0.MontageOutro.AnimStateTransitionNode_5.Transition
  nodes [Result(bCanEnterTransition=Get bUserCurve01())]
 }
 Transition {
  path=ItemFullBodyOverride.AnimGraphNode_StateMachine_0.MontageOutro.AnimStateTransitionNode_6.Transition
  nodes [Result(bCanEnterTransition=Get bChargedSwingHit() -> (bChargedSwingHit:bool))]
 }
 Transition {
  path=ItemFullBodyOverride.AnimGraphNode_StateMachine_0.MontageOutro.AnimStateTransitionNode_7.Transition
  nodes [Result(bCanEnterTransition=NOT Boolean(A=Get bChargedSwingHit()))]
 }
 Transition {
  path=ItemFullBodyOverride.AnimGraphNode_StateMachine_0.MontageOutro.AnimStateTransitionNode_8.Transition
  nodes [Result(bCanEnterTransition=Get bExitOutro() -> (bExitOutro:bool))]
 }
 IntoChargeState {
  entry=Base
  path=ItemFullBodyOverride.AnimGraphNode_StateMachine_1.IntoChargeState
  ids=id
  nodes [
   AnimStateNode_0 = Base() {is=title AnimStateNode}
   AnimStateNode_1 = IntoTacSprint() {is=title AnimStateNode}
   AnimStateTransitionNode_0 = AnimStateTransitionNode()
   AnimStateTransitionNode_1 = AnimStateTransitionNode()
  ]
  links [
   {
    dst=IntoTacSprint
    rule=ItemFullBodyOverride.AnimGraphNode_StateMachine_1.IntoChargeState.AnimStateTransitionNode_0.Transition
    src=Base
    via=AnimStateTransitionNode_0
   }
   {
    dst=Base
    rule=ItemFullBodyOverride.AnimGraphNode_StateMachine_1.IntoChargeState.AnimStateTransitionNode_1.Transition
    src=IntoTacSprint
    via=AnimStateTransitionNode_1
   }
  ]
 }
 Base {
  path=ItemFullBodyOverride.AnimGraphNode_StateMachine_1.IntoChargeState.AnimStateNode_0.Base
  nodes [Output Animation Pose(Result=Use cached pose 'FullBodyCache'())]
 }
 IntoTacSprint {
  path=ItemFullBodyOverride.AnimGraphNode_StateMachine_1.IntoChargeState.AnimStateNode_1.IntoTacSprint
  nodes [
   n1 = Layered blend per bone(
    BasePose=Sequence Player 'ElectroBat_Charged_JogToTacSprint_transition'() {bLoopAnimation=false, sequence=/ElectroBatGameplay/Assets/Animation/TacSprint/ElectroBat_Charged_JogToTacSprint_transition}
    BlendPoses_0=Use cached pose 'FullBodyCache'()
   ) {
    bMeshSpaceRotationBlend=true
    blendMode=BlendMask
    blendMasks [&Fortnite_M_Avg_Player_Skeleton:LowerBody]
    layerSetup []
   }
   Output Animation Pose(Result=@n1)
  ]
 }
 Transition {
  path=ItemFullBodyOverride.AnimGraphNode_StateMachine_1.IntoChargeState.AnimStateTransitionNode_0.Transition
  nodes [Result(bCanEnterTransition=Get bEnterChargedState() -> (bEnterChargedState:bool))]
 }
 Transition {
  path=ItemFullBodyOverride.AnimGraphNode_StateMachine_1.IntoChargeState.AnimStateTransitionNode_1.Transition
  nodes [Result(bCanEnterTransition=Get bUserCurve03() -> (bUserCurve03:bool))]
 }
 ItemIdleOverride {nodes [
   Input Pose() {name=InPose_Idle}
   Output Pose() {layerGroup=Item Override}
  ]}
 ItemJetPackJumpAdditive {nodes [Output Pose() {layerGroup=Item Override}]}
 ItemJetPackStartAdditive {nodes [Output Pose() {layerGroup=Item Override}]}
 ItemJumpLoopAdditive {nodes [
   Input Pose() {name=InPoseItemJumpLoopAdditive}
   Output Pose() {layerGroup=Item Override}
  ]}
 ItemJumpOverride {nodes [
   Input Pose() {name=InPoseJump}
   Output Pose() {layerGroup=Item Override}
  ]}
 ItemJumpUpAdditive {nodes [
   Input Pose() {name=InPoseItemJumpUpAdditive}
   Output Pose() {layerGroup=Item Override}
  ]}
 ItemPreAimOffsets {nodes [
   Input Pose() {name=InPosePreAimOffsets}
   Output Pose() {layerGroup=Item Override}
  ]}
 ItemPreIK {nodes [
   Input Pose() {name=InPose_PreIK}
   Output Pose() {layerGroup=Item Override}
  ]}
 ItemRightHandIK {nodes [
   Input Pose() {name=InPosePreRightHandIK}
   Input Pose() {name=InPosePostRightHandIK}
   Output Pose() {layerGroup=Item Override}
  ]}
 ItemSkeletalControl {nodes [
   Input Pose()
   Output Pose() {layerGroup=Item Override}
  ]}
 ItemSwimJumpFallLoop {nodes [Output Pose() {layerGroup=Item Override}]}
 ItemSwimJumpSurfaceEnd {nodes [Output Pose() {layerGroup=Item Override}]}
 ItemSwimJumpSurfaceLoop {nodes [Output Pose() {layerGroup=Item Override}]}
 ItemUpperBody {nodes [
   Output Pose(Result=Blend Poses by bool(
     BlendPose_0=Sequence Player 'ElectroBat_Sliding_Pose'() {bLoopAnimation=false, sequence=/Game/Animation/Game/MainPlayerFN/Poses/ElectroBat/ElectroBat_Sliding_Pose}
     BlendPose_1=Use cached pose 'UpperBodyCache'()
     bActiveValue=Get bIsSliding()
    ) {blendType=Linear}) {layerGroup=Item Override}
   Save cached pose 'UpperBodyCache'(Pose=Input Pose!() -> (Pose, InputParam:Fort Anim Input Adjusted Aim Structure) {name=InPoseUpperBody}) {cacheName=UpperBodyCache}
  ]}
 ItemZipLineStartAdditive {nodes [Output Pose() {layerGroup=Item Override}]}
]
variables []
