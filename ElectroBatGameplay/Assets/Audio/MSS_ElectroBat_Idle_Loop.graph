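# Mono looping MetaSound source for the ElectroBat's idle sound, with attenuation and concurrency limits.
# A wave player starts on play, shaped by random pitch and volume and an affiliation gain that makes enemies louder.
# An envelope follower on the ElectroBat duck audio bus drives an absolute one-minus-scaled-envelope gain, so bus activity ducks the loop, though strong peaks can raise it again.
# A repeating debug log node remains, unconnected to the envelope.
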MetaSoundSource /ElectroBatGameplay/Assets/Audio/MSS_ElectroBat_Idle_Loop
inputs [
 Character.Affiliation {type=Enum:FortAudioAffiliation}
 Character.Health {type=Float}
 Character.UniqueID {type=Int32}
 UE.Source.OnPlay {type=Trigger}
]
interfaces [Fort.Character:1.0, UE.OutputFormat.Mono:1.0, UE.Source:1.1]
outputs ["UE.OutputFormat.Mono.Audio:0" {type=Audio}]
pages [Default {
  input_defaults {Character.Affiliation=0, Character.Health=0, Character.UniqueID=0, UE.Source.OnPlay=null}
  nodes [
   AudioBusReader1_0 = "UE.Audio Bus Reader (1)"() {also {defaults {Audio Bus=&/AB_ElectroBat_Duck}}}
   EnvelopeFollower_0 = UE.Envelope Follower(In=@AudioBusReader1_0.Out 0)
   n1 = UE.Multiply:Float(PrimaryOperand=@EnvelopeFollower_0.Envelope) {also {defaults {AdditionalOperands=3}}}
   n2 = /FortMetasound/Patches/MSP_RandomPitchTimeVolume() {also {defaults {"Random Pitch Max (Semitones)"=0}}}
   n3 = &MSP_AffiliationMixFloat(Affiliation=^Character.Affiliation) {also {defaults {Enemy=1.5, Non-Enemy=0.65}}}
   WavePlayer_0 = UE.Wave Player:Mono(Play=^UE.Source.OnPlay, Pitch Shift=@n2.Random Pitch) {also {defaults {Loop="<MetaSoundBuilderResult.SUCCEEDED: 0>", Wave Asset=&/ElectroBat_Idle_Loop}}}
   Multiply_1 = UE.Multiply:Audio by Float(
    AdditionalOperands=AbsoluteValue.Abs:Float(Input=UE.Subtract:Float(AdditionalOperands=@n1.Out) {also {defaults {PrimaryOperand=1}}}.Out).Value
    PrimaryOperand=UE.Multiply:Audio by Float(
     AdditionalOperands=@n2.Random Volume
     PrimaryOperand=UE.Multiply:Audio by Float(AdditionalOperands=@n3.Out, PrimaryOperand=@WavePlayer_0.Out Mono).Out
    ).Out
   )
   PrintLogPrintLog_0 = Print Log.Print Log:Float(Trigger=UE.TriggerRepeat().RepeatOut) {also {defaults {Label=Bat Envelope}}}
  ]
  links [{dst="@UE.OutputFormat.Mono.Audio:0", src=Multiply_1.Out}]
 }]
properties {
 assetImportData=&self:AssetImportData
 attenuationSettings=&/ATT_ElectroBat_Idle
 bLooping=true
 duration=10000
 maxDistance=2500
 sampleRate=48000
 soundClassObject=&Fort_WeaponMelee
 volume=0.05
 concurrencyOverrides {maxCount=3}
 concurrencySet [&/CONC_ElectroBat_Idle_Loop_Global, &/CONC_ElectroBat_Idle_Loop_Local]
}
