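; ElectroBat weapon equip sound component that silences its audio while the player pawn matches the weapon override tag query.
; On BeginPlay it runs the parent logic, then resolves and stores the player pawn.
; An async gameplay tag query wait calls StopAudio when the override tags match, then starts a nested WhenFalse wait that calls PlayAudio when they clear.
; Tick is empty, so the behavior is entirely event-driven.
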(event EventBeginPlay
  (|CallParentFunction)
  (bind _playerpawn (Variables|Private|GetPlayerPawn)
    (:then
      (Variables|Default|SetPlayer%20Pawn _playerpawn)
      (Ability|Tasks|WaitGameplayTagQueryOnActor _playerpawn (Variables|Default|GetWeapon%20Override%20Tags)
        (:Triggered
          (CallFunction|StopAudio)
          (Ability|Tasks|WaitGameplayTagQueryOnActor (Variables|Default|GetPlayer%20Pawn) (Variables|Default|GetWeapon%20Override%20Tags) "WhenFalse"
            (:Triggered
              (CallFunction|PlayAudio))))))
    (:else)))

(event EventTick
  (params (DeltaSeconds :type Float%20%28single-precision%29)))
