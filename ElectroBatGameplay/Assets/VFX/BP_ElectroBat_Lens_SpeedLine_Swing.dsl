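; BP_ElectroBat_Lens_SpeedLine_Swing is a one-shot post-process lens effect for the ElectroBat, named for its swing.
; On BeginPlay, if player 0's pawn is within Detonate Radius, it creates a dynamic radial-blur material and plays a timeline.
; Each update feeds actor location, color rate and blur into the material, reapplies it as a blendable on PostProcess_RadialBlur, and animates a second PostProcess component's settings and blend weight.
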(event EventBeginPlay
  (bind _returnvalue (Transformation|GetActorLocation))
  (bind _dm (Variables|Default|GetDM))
  (bind _postprocess (Variables|Default|GetPostProcess))
  (if (not (> (Math|Vector|Distance(Vector) (Transformation|GetActorLocation (Game|GetPlayerPawn 0)) _returnvalue) (Variables|Default|GetDetonate%20Radius)))
    ; Inside
    (bind _returnvalue_2 (Rendering|Material|CreateDynamicMaterialInstance "/ElectroBatGameplay/Assets/VFX/Materials/PPM_ElectroBat_RadialBlur.PPM_ElectroBat_RadialBlur" "None" "None" :!class /Script/Engine.KismetMaterialLibrary))
    (Variables|Default|SetDM _returnvalue_2)
    (bind (_direction _newtrack_0 _newtrack_1) (Timeline|Timeline PlayFromStart)
      (:Update
        (Rendering|Material|SetVectorParameterValue _dm "WorldPosition" (Math|Color|MakeColor (.x _returnvalue) (.y _returnvalue) (.z _returnvalue)) :!class /Script/Engine.MaterialInstanceDynamic)
        (Rendering|Material|SetScalarParameterValue _dm "Add Color Rate" (Math|Float|Lerp 0.0 1.0 _newtrack_0) :!class /Script/Engine.MaterialInstanceDynamic)
        (Rendering|Material|SetScalarParameterValue _dm "BlurControl" (Math|Float|Lerp 0.0 0.25 _newtrack_1) :!class /Script/Engine.MaterialInstanceDynamic)
        (Variables|/Script/Engine.PostProcessComponent|SetSettings (Utilities|Struct|MakePostProcessSettings (Math|Float|Lerp 0.675 1.2 _newtrack_0)) _postprocess)
        (Variables|/Script/Engine.PostProcessComponent|SetBlendWeight _newtrack_1 _postprocess)
        (Variables|/Script/Engine.PostProcessComponent|SetSettings (Utilities|Struct|MakePostProcessSettings (Utilities|Struct|MakeWeightedBlendables (Utilities|Array|MakeArray (Utilities|Struct|MakeWeightedBlendable 1.0 (Variables|Default|GetDM))))) (Variables|Default|GetPostProcess_RadialBlur))))))

(fn ConstructionScript)
