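# Translucent deferred-decal material for the ElectroBat foot trail, flagged for Niagara ribbons.
# An opacity chain multiplies particle alpha, a sine falloff across V, squared plasma-coil noise projected on world XY, two UV1 sphere masks and inverted depth fade.
# A glow chain tints a UV-mapped plasma texture by particle color and a sharpened copy of the sine falloff.
# The InitialOffset dynamic parameter shifts both texture lookups per particle.
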Material /ElectroBatGameplay/Assets/VFX/Materials/MD_ElectroBat_Foot_Trail
graph {
 properties {
  bCastRayTracedShadows=false
  bUsedWithNiagaraRibbons=true
  blendMode=BLEND_Translucent
  materialDomain=MD_DeferredDecal
  pixelDepthOffsetMode=PDOM_Legacy
  translucencyLightingMode=TLM_SurfacePerPixelLighting
  usedShadingModels=Default Lit
 }
 nodes [
  Constant() {r=1}
  Constant() {r=1}
  n1 = ParticleColor()
  n2 = MaterialFunctionCall(Float2=TextureCoordinate()) {materialFunction=&BreakOutFloat2Components}
  n3 = Multiply(A=Sine(None=Multiply(A=@n2.G) {constB=0.5}), B=Constant() {r=0.5})
  n4 = Multiply(A=Multiply(A=@n1.A, B=@n3), B=ScalarParameter() {defaultValue=22, parameterName=Opacity})
  n5 = DynamicParameter() {paramNames [InitialOffset, Param2, Param3, Param4]}
  n6 = Multiply(A=@n5.InitialOffset) {constB=500}
  n7 = WorldPosition()
  n8 = Divide(A=ComponentMask(None=Add(A=@n6, B=@n7.XYZ)) {g=true, r=true}, B=Constant() {r=25})
  n9 = Constant() {r=5}
  n10 = TextureSample(UVs=Divide(A=Divide(A=@n8, B=@n9) {constB=5}, B=Constant() {r=5})) {samplerType=SAMPLERTYPE_LinearGrayscale, texture=/Game/Effects/Fort_Effects/Emotes/UltraLight/Texture/T_PlasmaCoils}
  n11 = Multiply(A=@n10.R, B=@n10.R)
  n12 = TextureCoordinate() {coordinateIndex=1}
  n13 = Constant() {r=3}
  n14 = Constant() {r=0.5}
  n15 = Constant() {r=-2}
  n16 = Multiply(
   A=SphereMask(A=@n12, B=AppendVector(A=@n13, B=@n14), Radius=Constant() {r=3}) {hardnessPercent=0}
   B=SphereMask(A=@n12, B=AppendVector(A=@n15, B=@n14), Radius=Constant() {r=3}) {hardnessPercent=0}
  )
  Multiply(A=Multiply(A=Multiply(A=@n4, B=@n11), B=Multiply(A=@n16) {constB=160}), B=OneMinus(None=DepthFade()))
  n17 = Multiply(A=Add(A=TextureCoordinate() {vTiling=2}, B=@n5.InitialOffset), B=AppendVector(A=Constant() {r=1}, B=Constant() {r=0.15}))
  Multiply(
   A=Multiply(
    A=@n1.RGB
    B=TextureSample(UVs=@n17) {samplerType=SAMPLERTYPE_LinearGrayscale, texture=/Game/Effects/Fort_Effects/Emotes/UltraLight/Texture/T_Plasma_A_B_Greyscale}.RGB
   )
   B=NamedRerouteUsage()
  )
  NamedRerouteDeclaration(None=Multiply(A=Power(Base=@n3) {constExponent=4}) {constB=55}) {name=Emissive Mask, nodeColor=<rgba 0.0078 0 1 1>}
  VectorParameter() {parameterName=Base, defaultValue=<rgba 0.099 0.0789 0.0653 1>}
 ]
}
