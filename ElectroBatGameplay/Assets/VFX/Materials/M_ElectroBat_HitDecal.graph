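# Translucent deferred decal for the ElectroBat hit, showing animated, glowing sine rings that travel outward from the actor's position.
# Time and world-space distance drive the sine, with its phase distorted by a scrolling noise texture sampled in 3D.
# A high power sharpens ring emissive, tinted between an outer color and the decal color by a center falloff.
# Opacity multiplies squared sine, that falloff, panning Voronoi volume noise and decal lifetime.
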Material /ElectroBatGameplay/Assets/VFX/Materials/M_ElectroBat_HitDecal
graph {
 properties {bAutomaticallySetUsageInEditor=false, blendMode=BLEND_Translucent, materialDomain=MD_DeferredDecal, usedShadingModels=Default Lit}
 nodes [
  n1 = Time()
  n2 = ScalarParameter() {defaultValue=-2, group=Sine, parameterName=Time}
  n3 = WorldPosition()
  n4 = Distance(A=ActorPositionWS(), B=@n3.XYZ)
  n5 = Multiply(A=@n4, B=ScalarParameter() {defaultValue=0.015, group=Sine, parameterName=Gradient})
  n6 = Add(
   A=Multiply(A=@n3.XYZ, B=ScalarParameter() {defaultValue=0.007, group=Noise, parameterName=Noise Distortion UV Scale})
   B=Multiply(A=Time(), B=ScalarParameter() {defaultValue=3, parameterName=Noise UV Anim Speed})
  )
  n7 = Multiply(
   A=Subtract(A=TextureSample(UVs=@n6) {samplerType=SAMPLERTYPE_LinearColor, texture=/Game/Global/MaterialFunctions/VT_ClothSimWPO}.R) {constB=0.5}
   B=ScalarParameter() {defaultValue=0.5, group=Noise, parameterName=Noise Distortion}
  )
  n8 = Sine(None=Add(A=Multiply(A=@n1, B=@n2), B=Add(A=@n5, B=@n7)))
  n9 = Power(Base=@n8, Exp=ScalarParameter() {defaultValue=555, group=Sine, parameterName=pow})
  n10 = DecalColor()
  n11 = Saturate(None=NamedRerouteUsage())
  n12 = Multiply(A=@n11, B=@n11)
  n13 = LinearInterpolate(A=VectorParameter() {parameterName=Outer Color, defaultValue=<rgba 0 0.1707 5 1>}.RGB, B=@n10.RGB, Alpha=Multiply(A=@n12, B=@n12))
  Add(A=Multiply(A=Multiply(A=@n9, B=@n13), B=ScalarParameter() {defaultValue=10, parameterName=Inner Intensity}), B=@n10.RGB)
  Constant() {r=1}
  Constant()
  n14 = Saturate(None=Multiply(A=@n4, B=ScalarParameter() {defaultValue=0.002, group=Sine, parameterName=Volume Sphere Center Mask Size}))
  n15 = Multiply(
   A=Power(Base=OneMinus(None=@n14), Exp=ScalarParameter() {defaultValue=14, group=Sine, parameterName="Mask Center Gradient  Pow"})
   B=ScalarParameter() {defaultValue=200, group=Sine, parameterName=Gradient Center Intensity}
  )
  n16 = MaterialFunctionCall(Float2=DecalLifetimeOpacity().Opacity) {materialFunction=&BreakOutFloat2Components}
  n17 = VectorParameter() {group=Volume Noise, parameterName=Volume Noise Pan, channelNames {a="", b=Z, g=Y, r=X}, defaultValue=<rgba 50 40 200 0>}
  n18 = Multiply(A=Time(), B=@n17)
  n19 = TextureSample(UVs=Multiply(A=Add(A=@n3.XYZ, B=@n18), B=ScalarParameter() {defaultValue=0.01, group=Volume Noise, parameterName=UVW_Scale})) {samplerType=SAMPLERTYPE_LinearColor, texture=/Game/Effects/Explosion_Update/Textures/VT_Voronoi_Seamless}
  n20 = ScalarParameter() {defaultValue=3, group=Volume Noise, parameterName=NoisePow}
  Multiply(A=Multiply(A=@n8, B=@n8), B=Multiply(A=Multiply(A=@n15, B=@n16.R), B=Power(Base=@n19.RGB, Exp=@n20)))
  NamedRerouteDeclaration(None=@n15) {name=Sine Mask Main, nodeColor=<rgba 1 0 0.9375 1>}
 ]
}
