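# Unlit translucent, two-sided Niagara sprite and mesh particle material for the ElectroBat ground impact, drawing a radial smoke burst.
# UVs, optionally rotated by RotateAmount, become radial coordinates that ExplosionPan pans through a smoke texture.
# Smoothstep sharpens the smoke, then sphere masks cut a ring, with Center Ctrl shaping the center subtraction.
# Particle color drives emissive through MF_FX_EmissiveAdjustments and its alpha scales the mask; the DistortAmount channel is unused.
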Material /ElectroBatGameplay/Assets/VFX/Materials/M_ElectroBat_ImpactGround
graph {
 properties {
  bUsedWithNiagaraMeshParticles=true
  bUsedWithNiagaraSprites=true
  blendMode=BLEND_Translucent
  pixelDepthOffsetMode=PDOM_Legacy
  shadingModel=MSM_Unlit
  twoSided=true
  usedShadingModels=Unlit
 }
 nodes [
  n1 = ParticleColor()
  MaterialFunctionCall(Emissive=@n1.RGB) {materialFunction=&MF_FX_EmissiveAdjustments}
  n2 = Add(A=TextureCoordinate(), B=Constant2Vector() {g=-0.5, r=-0.5})
  n3 = MaterialFunctionCall(
   UVs=@n2
   Rotation Center=Constant2Vector()
   "Rotation Angle (0-1)"=Multiply(A=DynamicParameter() {defaultValue=<rgba -1 0.1 0 3>, paramNames [ExplosionPan, DistortAmount, RotateAmount, Center Ctrl]}.RotateAmount) {constB=-1}
  ) {materialFunction=&CustomRotator}
  n4 = ScalarParameter() {defaultValue=1, parameterName=UseDistortion}
  n5 = Panner(
   Coordinate=MaterialFunctionCall(Vector or UVs=LinearInterpolate(A=@n2, B=@n3.Rotated Values, Alpha=@n4) {constB=0}) {materialFunction=&VectorToRadialValue}.Radial Coordinates
   Time=DynamicParameter() {defaultValue=<rgba -1 0.1 0 3>, paramNames [ExplosionPan, DistortAmount, RotateAmount, Center Ctrl]}.ExplosionPan
  ) {speedY=-0.25}
  n6 = TextureCoordinate()
  n7 = Constant2Vector() {g=0.5, r=0.5}
  n8 = Constant() {r=0.2}
  n9 = Constant() {r=0.4}
  n10 = Multiply(
   A=SmoothStep(
    Min=ScalarParameter() {defaultValue=0.5, parameterName=Min}
    Max=ScalarParameter() {defaultValue=1, parameterName=Max}
    Value=TextureSample(UVs=@n5) {texture=/Game/Effects/Fort_Effects/Textures/Smoke/T_Puffy_Smoke_Trail_01}.G
   ) {constMin=0.5}
   B=OneMinus(None=SphereMask(A=@n6, B=@n7, Radius=@n8, Hardness=@n9))
  )
  n11 = SphereMask(A=TextureCoordinate(), B=Constant2Vector() {g=0.5, r=0.5}, Radius=Constant() {r=0.4}, Hardness=Constant() {r=-1})
  n12 = DynamicParameter() {defaultValue=<rgba -1 0.1 0 3>, paramNames [ExplosionPan, DistortAmount, RotateAmount, Center Ctrl]}
  Multiply(A=@n1.A, B=Subtract(A=Multiply(A=@n10, B=@n11), B=Power(Base=@n11, Exp=@n12.Center Ctrl) {constExponent=3}))
 ]
}
