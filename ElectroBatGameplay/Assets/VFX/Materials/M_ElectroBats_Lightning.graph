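# Translucent, unlit, two-sided lightning material for Niagara mesh particles in the ElectroBats effects.
# World position offset thickens the spline mesh, wobbles it with panning curl noise, and stretches it toward a world point from particle dynamic parameters.
# A tiled lightning texture plus center glow, cut by an animated V wipe, forms the bolt in particle color.
# Quality switches vary emissive and opacity, with depth and near-camera fades on higher tiers.
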Material /ElectroBatGameplay/Assets/VFX/Materials/M_ElectroBats_Lightning
graph {
 properties {
  bAutomaticallySetUsageInEditor=false
  bUsedWithNiagaraMeshParticles=true
  blendMode=BLEND_Translucent
  shadingModel=MSM_Unlit
  twoSided=true
  usedShadingModels=Unlit
 }
 nodes [
  n1 = ComponentMask(None=TextureCoordinate()) {r=true}
  n2 = OneMinus(None=@n1)
  n3 = Power(Base=Multiply(A=Multiply(A=@n2, B=@n1)) {constB=2}, Exp=ScalarParameter() {defaultValue=4, parameterName=Glow Pow})
  n4 = Subtract(A=@n3, B=ScalarParameter() {defaultValue=0.007, parameterName=Core_Thickness}) {constB=0.01}
  n5 = Multiply(A=@n3, B=ScalarParameter() {defaultValue=0.25, parameterName=Glow Alpha})
  Add(A=Multiply(A=Saturate(None=@n4), B=ScalarParameter() {defaultValue=50, parameterName=Core}), B=@n5)
  n6 = VectorParameter() {parameterName=Thicken, channelNames {a=none, b=none, g=Tip, r=Base}, defaultValue=<rgba 20 20 0 0>}
  n7 = TextureCoordinate()
  n8 = Subtract(
   A=ComponentMask(None=OneMinus(None=@n7)) {g=true}
   B=DynamicParameter() {parameterIndex=1, defaultValue=<rgba 0 0 0 0>, paramNames [V_Anim,1,2,3]}.V_Anim
  )
  n9 = WorldPosition()
  n10 = Add(
   A=Subtract(A=WorldPosition().XYZ, B=ParticlePositionWS())
   B=Multiply(A=VectorParameter() {parameterName=Curl Noise Pan Dir, defaultValue=<rgba 250 400 -500 0>}.RGB, B=Time())
  )
  n11 = ScalarParameter() {defaultValue=0.001, parameterName=Noise Size}
  n12 = Subtract(A=TextureSample(UVs=Multiply(A=@n10, B=@n11)) {samplerType=SAMPLERTYPE_LinearColor, texture=/Game/Effects/FogVolume/VT_CurlNoise_LOW}.RGB) {constB=0.5}
  n13 = DynamicParameter() {defaultValue=<rgba 100 100 200 10>, paramNames [X, Y, Z, WPO Noise Strength]}
  n14 = ComponentMask(None=TextureCoordinate()) {g=true}
  n15 = Multiply(A=OneMinus(None=@n14), B=@n14)
  n16 = NamedRerouteUsage()
  n17 = DynamicParameter() {defaultValue=<rgba 100 100 200 10>, paramNames [X, Y, Z, WPO Noise Strength]}
  n18 = Add(
   A=Multiply(A=Multiply(A=@n12, B=@n13.WPO Noise Strength), B=Power(Base=@n15) {constExponent=0.5})
   B=Subtract(A=Multiply(A=@n16, B=@n17.RGB), B=Multiply(A=@n16, B=@n9.XYZ))
  )
  Add(
   A=MaterialFunctionCall(WidthBase=@n6.Base, WidthTip=Multiply(A=@n6.Tip, B=@n8), WorldPosition=@n9.XYZ, UVs for Thickness=TextureCoordinate()) {materialFunction=&SplineThicken}.WorldPositionOffset
   B=@n18
  )
  n19 = ParticleColor()
  n20 = Add(A=@n5, B=TextureSample(UVs=TextureCoordinate() {vTiling=2.4}) {texture=/Game/Athena/SafeZone/Textures/T_JanusShield_Lightning}.RGB)
  n21 = Sign(None=@n8)
  n22 = Multiply(A=@n19.RGB, B=Multiply(A=@n20, B=@n21))
  MaterialFunctionCall(Emissive=MaterialFunctionCall!(
    Low _ PC Low and Med=@n19.RGB
    "High _ XB1, PS4"=@n19.RGB
    "Epic _ PC High, XSX, PS5"=@n19.RGB
    --------=@n22
    "Mobile Low _ PC Perf, Android"=@n22
    Mobile Med _ Switch=@n22
   ) {materialFunction=/Game/Global/MaterialFunctions/TechArt/MF_Fortnite_QualitySwitch_Vector3}.Result) {materialFunction=&MF_FX_EmissiveAdjustments}
  MaterialFunctionCall!("Xb1,PS4"=@n18, "PC-High&&Epic,XSX,PS5"=@n18, "PC-Perf,Android"=@n18, Switch=@n18, "PC-Low(NotPerf)&&Medium"=@n18) {materialFunction=/Game/Effects/Niagara/MasterMaterials/Functions/Common/PlatformSwitch_3D}
  n23 = DepthFade(
   Opacity=Multiply(A=NamedRerouteUsage(), B=Multiply(A=@n19.A, B=@n20))
   FadeDistance=ScalarParameter() {defaultValue=200, parameterName=DepthFade}
  )
  Multiply(
   A=MaterialFunctionCall!(
    Low _ PC Low and Med=@n23
    "High _ XB1, PS4"=@n23
    "Epic _ PC High, XSX, PS5"=@n23
    --------=@n20
    "Mobile Low _ PC Perf, Android"=@n20
    Mobile Med _ Switch=@n20
   ) {materialFunction=/Game/Global/MaterialFunctions/TechArt/MF_Fortnite_QualitySwitch_Scalar}.Result
   B=@n21
  )
  NamedRerouteDeclaration(None=@n14) {name=WPO Mask, nodeColor=<rgba 1 0 0.1172 1>}
  n24 = MaterialFunctionCall(Distance=ScalarParameter() {defaultValue=400, parameterName=Near Camera Fade Distance}) {materialFunction=&Near_Camera_Fade}
  NamedRerouteDeclaration(None=VertexInterpolator(VS=Multiply(A=@n24.Result, B=@n24.Result)).PS) {name=Near Camera Fade, nodeColor=<rgba 0.6875 1 0 1>}
  Power(Base=MaterialFunctionCall(Distance=ScalarParameter() {defaultValue=700, parameterName=Near Camera Fade}) {materialFunction=&Near_Camera_Fade}.Result) {constExponent=0.54}
  VectorParameter() {parameterName=Glow Color, defaultValue=<rgba 0 38.7844 88 1>}
 ]
}
