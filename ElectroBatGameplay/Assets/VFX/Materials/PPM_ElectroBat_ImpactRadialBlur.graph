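# Post-process material for the ElectroBat impact effect: radial blur with a dark blue tint added toward the screen edges.
# An inverted screen-UV sphere mask drives the additive color and, presumably through its named reroute, the lerp between scene color and RadialBlur.
# An off-by-default static switch enables fallback to unblurred scene color on forward, mobile and Low quality.
# Despite that switch's "medium and low" name, Medium keeps the blur.
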Material /ElectroBatGameplay/Assets/VFX/Materials/PPM_ElectroBat_ImpactRadialBlur
graph {
 properties {materialDomain=MD_PostProcess, pixelDepthOffsetMode=PDOM_Legacy, usedShadingModels=Default Lit}
 nodes [
  n1 = Constant() {r=0.45}
  n2 = OneMinus(None=SphereMask(
    A=MaterialFunctionCall() {materialFunction=&ScreenAlignedUVs}."X 100%, Y 100%"
    B=@n1
    Radius=@n1
   ) {hardnessPercent=40})
  n3 = VectorParameter() {parameterName=Color, defaultValue=<rgba 0 0.0234 0.0938 1>}
  n4 = AppendVector(A=@n3.RGB, B=@n3.RGB)
  n5 = SceneTexture() {sceneTextureId=PPI_PostProcessInput0}
  n6 = MaterialFunctionCall!(IsStaticmesh) {materialFunction=/Game/Packages/Fortnite_UniqueMaterials_1/PostProcess/RadialBlur/RadialBlur}
  n7 = StaticSwitchParameter(True=LinearInterpolate(A=@n5.Color, B=@n6.Result, Alpha=NamedRerouteUsage()), False=@n6.Result) {defaultValue=true, parameterName=UseBlendLerp}
  n8 = ShadingPathSwitch(Default=@n7, Deferred=@n7, Forward=@n5.Color, Mobile=@n5.Color)
  Add(
   A=Multiply(A=Multiply(A=@n2, B=@n4), B=ScalarParameter() {defaultValue=1, parameterName=Add Color Rate})
   B=StaticSwitchParameter(True=QualitySwitch(Default=@n8, Low=@n5.Color, High=@n8, Medium=@n8, Epic=@n8), False=@n7) {parameterName=Not use on medium and low}
  )
  NamedRerouteDeclaration(None=@n2) {name=Sphere Mask, nodeColor=<rgba 0 1 0.9297 1>}
 ]
}
