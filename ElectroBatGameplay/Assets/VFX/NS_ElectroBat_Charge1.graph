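# Niagara charge-up effect for the ElectroBat, with Flare, WindCone and Lit emitters; only Lit shows a renderer, a light.
# Lit bursts one particle after a short loop delay.
# Curves on emitter age drive a rising Z position, a growing light radius and a color ramp to intense blue-cyan with sine-modulated brightness.
# The system exposes two charge-up finish times and an On_Water float, none referenced in the visible Lit logic.
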NiagaraSystem /ElectroBatGameplay/Assets/VFX/NS_ElectroBat_Charge1
dependencies {
 usedDataInterfaces [
  &NiagaraDataInterfaceVectorCurve
  &NiagaraDataInterfaceColorCurve
  &NiagaraDataInterfaceCurve
  &NiagaraDataInterfaceVector2DCurve
 ]
 usedDynamicInputs [
  &Subtract_Float
  &Add_Float
  &RandomRangeFloat
  &Multiply_VectorByFloat
  &ScaleFloatByCurve
  &Divide_Float
  &Vector2DFromCurve
 ]
 usedModules [
  &self:SystemSpawnScript.SystemScriptSource.SystemScriptGraph.NiagaraNodeAssignment_1.SetVariables_#1
  &SystemState
  &self:SystemSpawnScript.SystemScriptSource.SystemScriptGraph.NiagaraNodeAssignment_0.SetVariables_#2
 ]
}
emitters {
 Flare {rendererClasses [], renderers []}
 Lit {
  emitterUpdateScript [
   &EmitterState(Life Cycle Mode=::NewEnumerator1, Loop Behavior=::NewEnumerator1, Loop Delay=0.1, UseLoopDelay=true)
   /SetVariables(Emitter.Position2:NiagaraPosition=&VectorFromCurve(
     VectorCurve=DataInterface() {propertyValues {
       XCurve {keys [<at 1 0>]}
       YCurve {keys [<at 1 0>]}
       ZCurve {keys [{tangentMode=RCTM_User, time=0, value=0}, {tangentMode=RCTM_User, time=1, value=1}]}
      }}
     CurveIndex=$Emitter.Age
     Scale Curve=<0 0 40>
    )) {isSetParameters=true}
   &SpawnBurst_Instantaneous(Spawn Count=1, Use Spawn Probability=false)
   /SetVariables(
    Emitter.LightColor:LinearColor=&Multiply_LinearColorByFloat(
     Linear Color=&ColorFromCurve(
      ColorCurve=DataInterface() {propertyValues {
        AlphaCurve {keys [<at 0 0>,<at 0.7772 0.5 RCIM_Cubic>]}
        BlueCurve {keys [
          <at 0 0>
          {arriveTangent=26.6667, leaveTangent=26.6667, time=0.6755, value=10}
          {arriveTangent=92.455, interpMode=RCIM_Cubic, leaveTangent=92.455, time=1, value=80}
          {interpMode=RCIM_Cubic, leaveTangent=45.3007, time=1, value=40}
         ]}
        GreenCurve {keys [<at 0 0>, {leaveTangent=0.1316, time=0.6755, value=2.6974}, <at 1 35>, <at 1 10.7895>]}
        RedCurve {keys [<at 0 0>, {leaveTangent=0.0359, time=0.6755, value=0.3646}]}
       }}
      CurveIndex=$Emitter.Age
     )
     Float=&Sine(Normalized Angle=$Emitter.Age, Period=0.5, Scale=0.1, Bias=1)
    )
    Emitter.LightRadius:NiagaraFloat=&FloatFromCurve(
     FloatCurve=DataInterface() {propertyValues {Curve {keys [
         {arriveTangent=0.0789, interpMode=RCIM_Cubic, leaveTangent=0.0789, tangentMode=RCTM_User, time=0, value=0}
         {arriveTangent=1.4286, interpMode=RCIM_Cubic, leaveTangent=1.4286, time=0.2, value=0.7}
         <at 0.7 1>
         <at 0.9 1>
        ]}}}
     CurveIndex=$Emitter.Age
     Scale Curve=70
    )
   ) {isSetParameters=true}
  ]
  rendererClasses [&NiagaraLightRendererProperties]
  renderers [{rendererClass=&NiagaraLightRendererProperties, rendererIndex=0}]
 }
 WindCone {rendererClasses [], renderers []}
}
summary {
 systemName=NS_ElectroBat_Charge1
 emitters [
  {emitterName=Flare, rendererClasses []}
  {emitterName=Lit, rendererClasses [&NiagaraLightRendererProperties]}
  {emitterName=WindCone, rendererClasses []}
 ]
 userVariables [
  User.ChargeUpFinishTime {type=NiagaraFloat, defaultValue=1, description=""}
  User.ChargeUpFinishTime2 {type=NiagaraFloat, defaultValue=1, description=""}
  User.On_Water {type=NiagaraFloat, defaultValue=1, description=""}
 ]
}
