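# Niagara effect for the ElectroBat's charge-up, with blue light, lightning and falling sparks.
# Lit ramps a flickering light and rotating mesh, while Lightning spawns short-lived mesh bolts after a User.ChargeUpFinishTime delay.
# Epic_High (GPU) and MedLow emit colliding mesh/sprite sparks that sample skeletal-mesh positions into material parameters, gate gravity, scale and alpha by speed, and use owner distance to control lifetime.
# Remaining emitters list no renderers.
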NiagaraSystem /ElectroBatGameplay/Assets/VFX/NS_ElectroBat_Charge2
dependencies {
 usedDataInterfaces [
  &NiagaraDataInterfaceCurve
  &NiagaraDataInterfaceColorCurve
  &NiagaraDataInterfaceArrayFloat3
  &NiagaraDataInterfaceSkeletalMesh
  &NiagaraDataInterfaceVectorCurve
  &NiagaraDataInterfaceVector2DCurve
 ]
 usedDynamicInputs [
  &Subtract_Float
  &ScaleFloatByCurve
  &Divide_Float
  &Vector2DFromCurve
 ]
 usedModules [
  &self:SystemSpawnScript.SystemScriptSource.SystemScriptGraph.NiagaraNodeAssignment_1.SetVariables_#1
  &SystemState
  &self:SystemSpawnScript.SystemScriptSource.SystemScriptGraph.NiagaraNodeAssignment_0.SetVariables_#2
 ]
}
emitters {
 Dark {rendererClasses [], renderers []}
 Epic_High {
  simTarget=GPUComputeSim
  emitterUpdateScript [
   &EmitterState(
    Delay First Loop Only=true
    Inactive Response=::NewEnumerator1
    Life Cycle Mode=::NewEnumerator1
    Loop Delay=0.5
    Loop Duration=1
    UseLoopDelay=true
   )
   &SpawnRate(
    SpawnRate=&RandomRangeFloat(Minimum=-3, Maximum=8, Evaluation Type=::NewEnumerator1)
    Use Spawn Probability=false
   )
   &SpawnBurst_Instantaneous(Spawn Count=4, Use Spawn Probability=false, UseLoopCountLimit=true) {enabled=false}
  ]
  particleSpawnScript [
   &InitializeParticle(
    Color=<rgba 0 0.2943 1 1>
    Lifetime Max=0.75
    Lifetime Min=0.65
    Lifetime Mode=::NewEnumerator1
    Mesh Scale Mode=::NewEnumerator3
    Mesh Uniform Scale=&Multiply_Float(A=&RandomRangeFloat(Minimum=0.18, Maximum=0.21), B=0.25)
    Position=$Engine.Owner.Position
    Position Mode=::NewEnumerator1
    Sprite Size Max=<70 50>
    Sprite Size Min=<65 45>
    Sprite Size Mode=::NewEnumerator2
    UsePositionOffset=false
    Write Material Random=true
   ) {module=InitializeParticle001}
   &ShapeLocation(
    Cylinder Height=50
    Cylinder Radius=5
    MaskSpawnGroup=false
    Non Uniform Scale=<1 1.2 1>
    Offset=<0 0 20>
    Offset Coordinate Space=::Local
    Offset Mode=::NewEnumerator0
    Override Seed=false
    Shape Primitive=::NewEnumerator1
    Surface Only=false
    UseOwnerScale=true
   ) {enabled=false}
   /SetVariables(Particles.SpriteMesh_Quat:Quat4f=/Niagara/DynamicInputs/Rotation/RandomQuaternion(), Particles.Sprite_CamOfs=25) {isSetParameters=true, module=SetVariables_#3}
   &InitialMeshOrientation(
    Mesh Orientation Mode=::NewEnumerator1
    Rotation=&RandomRangeVector(Minimum=<-0.1 -0.1 -0.1>, Maximum=<0.1 0.1 0.1>)
    Use Rotation Vector=true
   )
   /SetVariables(
    Particles.InitPos:NiagaraPosition=$Particles.Initial.Position
    Particles.StoredVector:Vector3f=&TransformVector(
     Vector=Linked() {linkedVariable {name=Particles.Initial.Position, type=NiagaraPosition}}
     Source Space=::Simulation
     Destination Space=::World
     Apply Scale=true
     Transform Vector as Position=true
    )
   ) {enabled=false, isSetParameters=true, module=SetVariables_#4}
   &AddVelocity(Rotation Coordinate Space=::Simulation, Velocity=<0 0 1>, Velocity Speed Scale=200) {enabled=false}
   &AddVelocity(
    Velocity=&Multiply_Vector(
     A=&RandomRangeVector(Minimum=<0.6 0.6 0.6>, Maximum=<0.9 0.9 1>)
     B=&SelectVectorFromArray(
      Array Sampling Mode=::NewEnumerator0
      Direct Array Index=?
      Vector Selection Array=DataInterface() {propertyValues {FloatData [<-100 -100 -50>, <100 100 -60>]}}
      Random Seed=0
     )
    )
    Velocity Speed Scale=20
   ) {module=AddVelocity001}
  ]
  particleUpdateScript [
   &ParticleState()
   &SkeletalMeshLocation(
    Bone Sampling Mode=::NewEnumerator4
    MaskSpawnGroup=false
    Orientation Sampling=::NewEnumerator1
    Position Sampling=::NewEnumerator1
    Sampled Position Offset=<20 0 -77>
    Skeletal Mesh=DataInterface() {propertyValues {
      DefaultMesh=null
      ExcludeBoneName=root
      OptionalComponentClass=null
      PreviewMesh=/Game/Characters/Player/Male/Medium/Base/SK_M_MALE_Base_Skeleton
      SkinningMode=SkinOnTheFly
      SoftSourceActor=null
      SourceMode=Default
      UvSetIndex=0
      WholeMeshLOD=-1
      bExcludeBone=false
      bReadDeformedGeometry=true
      bRequireCurrentFrameData=true
      ComponentTags []
      FilteredBones [hand_l]
      FilteredSockets [FX_AnimTrail_01, FX_AnimTrail_02]
      MeshUserParameter {parameter {name=None, typeDefHandle {registeredTypeIndex=1}, varData []}}
      SamplingRegions []
     }}
    Source Movement Scale:Vector3f=&VectorFromCurve(VectorCurve=DataInterface() {propertyValues {XCurve {keys [<at 0 1>,<at 0.1 0>]}, YCurve {keys [<at 0 1>,<at 0.1 0>]}, ZCurve {keys [<at 0 1>,<at 0.1 0>]}}})
   )
   &UpdateMeshOrientation(Rotation Rate=&RandomRangeFloat(Minimum=-1, Maximum=1), Rotation Vector=<0.15 0.15 0.1>)
   &GravityForce(
    Coordinate Space=::Local
    Gravity=&Multiply_VectorByFloat(
     Vector=&AddVectorToPosition(Position=&RandomRangeVector(Minimum=<0 0 -35000>, Maximum=<0 0 -35000>), Vector=<0 0 0>)
     Float=&MakeCustomFloatFromBool(
      Bool=&SetBoolByFloatComparison(A=5, B=HlslExpression() {hlslExpression="length(Particles.Velocity)"})
      True Float=0
      False Float=1
     )
    )
   )
   &Drag(
    Drag=&FloatFromCurve(
     FloatCurve=DataInterface() {propertyValues {Curve {keys [{tangentMode=RCTM_User, time=0, value=0}, {tangentMode=RCTM_User, time=0.4, value=1}]}}}
     Scale Curve=90
    )
    Use Linear Drag=true
    Use Rotational Drag=true
   ) {enabled=false}
   &Collision(
    Age Colliding Particles=false
    Collision Enabled=&SetBoolByFloatComparison(
     A=1
     B=&FloatFromCurve(FloatCurve=DataInterface() {propertyValues {Curve {keys [<at 0.05 0.95>,<at 0.5 2>]}}}, Scale Curve=0.5)
    )
    Friction=10000
    Friction During a Bounce=10000
    GPU Collision Type=::NewEnumerator2
    Maximum Penetration Correction Distance=1
    Method for Calculating Particle Radius:ENiagaraCalculateRadiusOptions=::NewEnumerator2
    Module.Control Rest State Via Surface Normal=false
    Module.EnableMaxCollisionCount=false
    Module.Randomize Collision Normal Vector=false
    Particle Radius Scale=0.12
    Randomize Collision Normal=0
    Rest State Time Range=1
    Restitution=0
   )
   &SolveForcesAndVelocity(Clamp Velocity=false, Limit Acceleration=false)
   &ScaleSpriteSize(
    Uniform Curve Scale=1.2
    Uniform Curve Sprite Scale=DataInterface() {propertyValues {Curve {keys [{tangentMode=RCTM_User, time=0, value=0.2}, <at 0.5 1>, {tangentMode=RCTM_User, time=0.7, value=0}]}}}
   )
   &ScaleMeshSize(Scale Factor=&Multiply_VectorByFloat(
     Vector=&MakeVector(
      X=1
      Y=1
      Z=&ClampFloat(Float=&Multiply_Float(A=10, B=HlslExpression() {hlslExpression="length(Particles.Velocity)"}), Min=2, Max=3)
     )
     Float=&MakeCustomFloatFromBool(
      Bool=&SetBoolByFloatComparison(A=10, B=HlslExpression() {hlslExpression="length(Particles.Velocity)"})
      True Float=&FloatFromCurve(
       FloatCurve=DataInterface() {propertyValues {Curve {keys [
           {tangentMode=RCTM_User, time=0, value=0.4}
           <at 0.1 0.45>
           <at 0.2 0.35>
           <at 0.3 0.45>
           <at 0.3636 0.4>
           <at 0.4246 0.4>
           <at 0.4803 0.6>
           <at 0.5186 0.25>
           <at 0.5761 0.4>
           <at 0.6533 0.25>
           <at 0.7 0.5236>
           <at 0.7391 0.183>
           <at 0.8 0.45>
           <at 0.8508 0.2508>
           <at 0.8845 0.5723>
           <at 0.9139 0.25>
           <at 0.9483 0.4763>
           <at 0.9912 0.5>
          ]}}}
       Scale Curve=2
      )
      False Float=0
     )
    ))
   &DynamicMaterialParameters(
    Index 0 Param 1=HlslExpression() {hlslExpression=Particles.SkeletalMeshLocation.SampledPosition.x}
    Index 0 Param 2=HlslExpression() {hlslExpression=Particles.SkeletalMeshLocation.SampledPosition.y}
    Index 0 Param 3=HlslExpression() {hlslExpression=Particles.SkeletalMeshLocation.SampledPosition.z}
    Index 0 Param 4=20
    Index 1 Param 1=&FloatFromCurve(FloatCurve=DataInterface() {propertyValues {Curve {keys [<at 0.05 1 RCIM_Cubic>, {arriveTangent=-1.1493, tangentMode=RCTM_User, time=0.5, value=0}]}}})
    Module.Param0WriteEnabled=true
    Module.Param0WriteEnabled1=true
    Module.Param1WriteEnabled=true
    Module.Param1WriteEnabled1=false
    Module.Param2WriteEnabled=true
    Module.Param2WriteEnabled1=false
    Module.Param3WriteEnabled=true
    Module.Param3WriteEnabled1=false
    Write Parameter Index 1=true
   ) {module=DynamicMaterialParameters001}
   &ScaleColor(
    Scale Alpha=&MakeCustomFloatFromBool(
     Bool=&SetBoolByFloatComparison(A=0.2, B=HlslExpression() {hlslExpression="length(Particles.Velocity)"})
     True Float=&FloatFromCurve(
      FloatCurve=DataInterface() {propertyValues {Curve {keys [
          <at 0.1 0.65 RCIM_Cubic>
          {arriveTangent=-0.0405, interpMode=RCIM_Cubic, leaveTangent=-0.0405, tangentMode=RCTM_User, time=0.4, value=1.75}
          <at 0.9 0 RCIM_Cubic>
         ]}}}
      Scale Curve=&FNTimeOfDayScalar(Night Scale=0.6, Morning Scale=0.7, Evening Scale=0.7)
     )
     False Float=0
    )
    Scale RGB=<4 4 4>
    ScaleA=true
    ScaleRGB=true
   )
   /SetVariables(
    Particles.Color2=&Multiply_LinearColorByFloat(
     Linear Color=&ColorFromCurve(ColorCurve=DataInterface() {propertyValues {
        AlphaCurve {keys [
          <at 0 0>
          {arriveTangent=31.4153, interpMode=RCIM_Cubic, leaveTangent=31.4153, time=0.107, value=15}
          <at 0.4775 15>
          <at 0.6729 0 RCIM_Cubic>
         ]}
        BlueCurve {keys [<at 0 8.2292 RCIM_Cubic>]}
        GreenCurve {keys [<at 0 2.5393 RCIM_Cubic>]}
        RedCurve {keys [<at 0 0.1714 RCIM_Cubic>]}
       }})
     Float=&MakeCustomFloatFromBool(
      Bool=&SetBoolByFloatComparison(A=10, B=HlslExpression() {hlslExpression="length(Particles.Velocity)"})
      True Float=1
      False Float=0
     )
    )
    Particles.Color3:LinearColor=&ColorFromCurve(ColorCurve=DataInterface() {propertyValues {
       AlphaCurve {keys [<at 0.0514 15 RCIM_Cubic>,<at 0.8816 0 RCIM_Cubic>]}
       BlueCurve {keys [<at 0 200 RCIM_Cubic>]}
       GreenCurve {keys [<at 0 86.3046 RCIM_Cubic>]}
       RedCurve {keys [<at 0 0 RCIM_Cubic>]}
      }})
    Particles.DM2_0:Vector4f=&MakeVector4(
     X=&Multiply_Float(
      A=&FloatFromCurve(
       FloatCurve=DataInterface() {propertyValues {Curve {keys [{tangentMode=RCTM_User, time=0, value=1}, {tangentMode=RCTM_User, time=0.8, value=0}]}}}
       Scale Curve=4
      )
      B=&MakeCustomFloatFromBool(
       Bool=&SetBoolByFloatComparison(A=5, B=HlslExpression() {hlslExpression="length(Particles.Velocity)"})
       True Float=0.2
       False Float=0
      )
     )
     Y=&FloatFromCurve(
      FloatCurve=DataInterface() {propertyValues {Curve {keys [
          {arriveTangent=2.2144, interpMode=RCIM_Cubic, leaveTangent=2.2144, tangentMode=RCTM_User, time=0, value=0}
          {tangentMode=RCTM_User, time=0.9, value=1}
         ]}}}
      Scale Curve=4
     )
     Z=&FloatFromCurve(
      FloatCurve=DataInterface() {propertyValues {Curve {keys [
          {arriveTangent=3.2315, interpMode=RCIM_Cubic, leaveTangent=3.2315, tangentMode=RCTM_User, time=0, value=0}
          {arriveTangent=1.1179, interpMode=RCIM_Cubic, leaveTangent=1.1179, tangentMode=RCTM_User, time=0.4, value=1}
         ]}}}
      Scale Curve=45
     )
     W=&FloatFromCurve(
      FloatCurve=DataInterface() {propertyValues {Curve {keys [
          {arriveTangent=0.6176, interpMode=RCIM_Cubic, leaveTangent=0.6176, tangentMode=RCTM_User, time=0.2, value=0}
          {arriveTangent=6.707, interpMode=RCIM_Cubic, leaveTangent=6.707, tangentMode=RCTM_User, time=1, value=2.5}
         ]}}}
      Scale Curve=-250
     )
    )
    Particles.Mesh Scale2:Vector3f=<10 10 10>
   ) {isSetParameters=true, module=SetVariables_#5}
   /SetVariables(Particles.Lifetime=&MakeCustomFloatFromBool(
     Bool=&SetBoolByFloatComparison(A=800, B=&DistanceBetweenPositions(A=$Engine.Owner.Position, B=$Particles.Position))
     True Float=$Particles.Lifetime
     False Float=0
    )) {isSetParameters=true, module=SetVariables_#6}
  ]
  rendererClasses [&NiagaraMeshRendererProperties, &NiagaraSpriteRendererProperties]
  renderers [
   {rendererClass=&NiagaraMeshRendererProperties, rendererIndex=0}
   {rendererClass=&NiagaraSpriteRendererProperties, rendererIndex=1}
   {rendererClass=&NiagaraMeshRendererProperties, rendererIndex=2}
  ]
 }
 Flare {rendererClasses [], renderers []}
 Flare002 {rendererClasses [], renderers []}
 Flare003 {rendererClasses [], renderers []}
 LW_Lightning001 {rendererClasses [], renderers []}
 Lightning {
  emitterUpdateScript [
   &EmitterState(
    Inactive Response=::NewEnumerator1
    Life Cycle Mode=::NewEnumerator1
    Loop Behavior=::NewEnumerator1
    Loop Delay=$User.ChargeUpFinishTime
    Loop Duration Mode=::NewEnumerator1
    UseLoopDelay=true
   )
   &SpawnRate(
    SpawnRate=&RandomRangeFloat(Minimum=10, Maximum=20, Evaluation Type=::NewEnumerator1)
    Use Spawn Probability=false
   )
  ]
  particleSpawnScript [
   &InitializeParticle(
    Color Mode=::NewEnumerator0
    Lifetime Max=0.18
    Lifetime Min=0.15
    Lifetime Mode=::NewEnumerator1
    Mesh Scale Max=<0.3 0.3 0.4>
    Mesh Scale Min=<0.2 0.2 0.3>
    Mesh Scale Mode=::NewEnumerator2
    UsePositionOffset=false
    Write Material Random=true
   )
   &InitialMeshOrientation(
    Mesh Orientation Mode=::NewEnumerator2
    Rotation=&RandomRangeVector(Minimum=<0 0 -0.5>, Maximum=<1 0 0.05>)
    Use Rotation Vector=true
   )
   &ShapeLocation(
    Box Midpoint=<0.5 0.5 0>
    Box Size=<0 0 65>
    MaskSpawnGroup=false
    Offset=<0 0 5>
    Offset Coordinate Space=::Local
    Offset Mode=::NewEnumerator0
    Override Seed=false
    Shape Primitive=::NewEnumerator2
    Surface Only=false
    UseOwnerScale=true
   )
   /SetVariables(Particles.ScaleByLength:NiagaraFloat=&DistanceBetweenPositions(A=$Particles.Position, B=$Engine.Emitter.SimulationPosition)) {isSetParameters=true}
  ]
  particleUpdateScript [
   &ParticleState()
   &ScaleMeshSize(Scale Factor=&VectorFromCurve(
     VectorCurve=DataInterface() {propertyValues {XCurve {keys [<at 0 1>,<at 1 1>]}, YCurve {keys [<at 0 1>,<at 1 1>]}, ZCurve {keys [<at 0 1>,<at 1 1.02>]}}}
     Scale Curve=&AddVectorToPosition(
      Position=&Multiply_VectorByFloat(Vector=<0.02 0.02 0.02>, Float=$Particles.ScaleByLength)
      Vector=<0.15 0.15 0.26>
     )
    ))
   &DynamicMaterialParameters(
    Index 0 Param 4=1
    Index 1 Param 1=&FloatFromCurve(FloatCurve=DataInterface() {propertyValues {Curve {keys [
         {arriveTangent=3.3469, interpMode=RCIM_Cubic, leaveTangent=3.3469, tangentMode=RCTM_User, time=0.1, value=0}
         {arriveTangent=0.0224, interpMode=RCIM_Cubic, leaveTangent=0.0224, tangentMode=RCTM_User, time=0.7, value=1.2}
        ]}}})
    Index 1 Param 3=&FloatFromCurve(FloatCurve=DataInterface() {propertyValues {Curve {keys [{interpMode=RCIM_Cubic, tangentMode=RCTM_Break, time=0.5, value=-0.5}, <at 1 0.1>]}}})
    Index 2 Param 1=&FloatFromCurve(
     FloatCurve=DataInterface() {propertyValues {Curve {keys [{interpMode=RCIM_Cubic, leaveTangent=0.3637, tangentMode=RCTM_Break, time=0.7, value=0.05}, {tangentMode=RCTM_User, time=1, value=0}]}}}
     Scale Curve=&Add_Float(A=&Multiply_Float(A=0.012, B=$Particles.ScaleByLength), B=0.14)
    )
    Index 2 Param 2=&FloatFromCurve(FloatCurve=DataInterface() {propertyValues {Curve {keys [
         {interpMode=RCIM_Cubic, tangentMode=RCTM_Break, time=0.7, value=0}
         {arriveTangent=-0.0859, interpMode=RCIM_Cubic, tangentMode=RCTM_Break, time=1, value=-50}
        ]}}})
    Index 2 Param 3=&RandomRangeFloat(Minimum=0.4, Maximum=0.6)
    Index 2 Param 4=25
    Index 3 Param 1=&FloatFromCurve(FloatCurve=DataInterface() {propertyValues {Curve {keys [
         {interpMode=RCIM_Cubic, leaveTangent=-0.0025, tangentMode=RCTM_Break, time=0, value=0}
         {arriveTangent=1.8528, interpMode=RCIM_Cubic, tangentMode=RCTM_Break, time=1, value=20}
        ]}}})
    Index 3 Param 2=&FloatFromCurve(FloatCurve=DataInterface() {propertyValues {Curve {keys [
         {interpMode=RCIM_Cubic, leaveTangent=-0.0025, tangentMode=RCTM_Break, time=0, value=0.5}
         {arriveTangent=-0.0316, interpMode=RCIM_Cubic, tangentMode=RCTM_Break, time=1, value=-0.2}
        ]}}})
    Index 3 Param 3=&FloatFromCurve(FloatCurve=DataInterface() {propertyValues {Curve {keys [
         {interpMode=RCIM_Cubic, leaveTangent=1.1034, tangentMode=RCTM_Break, time=0, value=-1}
         {arriveTangent=2.2399, tangentMode=RCTM_User, time=0.5, value=0.125}
        ]}}})
    Index 3 Param 4=&FloatFromCurve(FloatCurve=DataInterface() {propertyValues {Curve {keys [
         {interpMode=RCIM_Cubic, leaveTangent=3.0601, tangentMode=RCTM_Break, time=0.5, value=-0.5}
         {arriveTangent=3.2053, tangentMode=RCTM_User, time=1, value=1}
        ]}}})
    Module.Index0Param0Write=true
    Module.Index0Param1Write=true
    Module.Index0Param2Write=true
    Module.Index0Param3Write=true
    Module.Index1Param0Write=true
    Module.Index1Param1Write=true
    Module.Index1Param2Write=true
    Module.Index1Param3Write=true
    Module.Index2Param0Write=true
    Module.Index2Param1Write=true
    Module.Index2Param2Write=true
    Module.Index2Param3Write=true
    Module.Index3Param0Write=true
    Module.Index3Param1Write=true
    Module.Index3Param2Write=true
    Module.Index3Param3Write=true
    Write Parameter Index 1=true
    Write Parameter Index 2=true
    Write Parameter Index 3=true
   )
  ]
  rendererClasses [&NiagaraMeshRendererProperties]
  renderers [
   {rendererClass=&NiagaraMeshRendererProperties, rendererIndex=0}
   {rendererClass=&NiagaraMeshRendererProperties, rendererIndex=1}
  ]
 }
 Lit {
  emitterSpawnScript [/SetVariables(Emitter.CamOfs=50, Emitter.Position2:NiagaraPosition=<-10 0 80>) {isSetParameters=true}]
  emitterUpdateScript [
   &EmitterState(
    Inactive Response=::NewEnumerator1
    Life Cycle Mode=::NewEnumerator1
    Loop Behavior=::NewEnumerator1
    Loop Delay=0.1
    Loop Duration Mode=::NewEnumerator1
    UseLoopDelay=false
   )
   &SpawnRate(
    SpawnRate=&FloatFromCurve(FloatCurve=DataInterface() {propertyValues {Curve {keys [<at 0 1 RCIM_Constant>,<at 1 0>]}}}, CurveIndex=$Emitter.Age, Scale Curve=6)
    Use Spawn Probability=false
   )
   /SetVariables(
    Emitter.LightColor:LinearColor=&Multiply_LinearColorByFloat(
     Linear Color=&ColorFromCurve(
      ColorCurve=DataInterface() {propertyValues {
        AlphaCurve {keys [<at 0 0>,<at 0.7772 1 RCIM_Cubic>]}
        BlueCurve {keys [<at 0 0>, {interpMode=RCIM_Cubic, leaveTangent=0.431, time=0.6755, value=70}]}
        GreenCurve {keys [<at 0 0>, {interpMode=RCIM_Cubic, leaveTangent=0.1316, time=0.6755, value=21.3686}]}
        RedCurve {keys [<at 0 0>, {interpMode=RCIM_Cubic, leaveTangent=0.0359, time=0.6755, value=5.8333}]}
       }}
      CurveIndex=$Emitter.Age
     )
     Float=&Sine(Normalized Angle=$Emitter.Age, Period=0.08, Scale=0.2, Bias=1)
    )
    Emitter.LightRadius:NiagaraFloat=&FloatFromCurve(
     FloatCurve=DataInterface() {propertyValues {Curve {keys [
         {arriveTangent=0.0789, interpMode=RCIM_Cubic, leaveTangent=0.0789, tangentMode=RCTM_User, time=0, value=0}
         {arriveTangent=1.4286, interpMode=RCIM_Cubic, leaveTangent=1.4286, time=0.2, value=0.7}
         <at 0.7 1>
        ]}}}
     CurveIndex=$Emitter.Age
     Scale Curve=150
    )
   ) {isSetParameters=true}
  ]
  particleSpawnScript [
   &InitializeParticle(
    Color=<rgba 0.25 5.2955 16 4>
    Lifetime=0.25
    Mesh Scale=<0.7 0.7 0.5>
    Mesh Scale Mode=::NewEnumerator4
    Position Offset=&MakeVector(
     X=0
     Y=0
     Z=&Add_Float(
      A=10
      B=&FloatFromCurve(
       FloatCurve=DataInterface() {propertyValues {Curve {keys [{tangentMode=RCTM_User, time=0, value=0}, {tangentMode=RCTM_User, time=1, value=1}]}}}
       CurveIndex=$Emitter.Age
       Scale Curve=60
      )
     )
    )
    Position Offset Coordinate Space=::Local
    UsePositionOffset=true
    Write Material Random=true
   )
   &InitialMeshOrientation(
    Mesh Orientation Mode=::NewEnumerator2
    Rotation=&RandomRangeVector(Minimum=<0 0 0>, Maximum=<0 0 1>)
    Use Rotation Vector=true
   )
  ]
  particleUpdateScript [
   &ParticleState()
   &UpdateMeshOrientation(Rotation Rate=2, Rotation Vector=<0 0 1>)
   &ScaleMeshSize(Scale Factor=&Multiply_VectorByFloat(
     Vector=<4 4 5>
     Float=&FloatFromCurve(
      FloatCurve=DataInterface() {propertyValues {Curve {keys [<at 0 0.33 RCIM_Cubic>, {arriveTangent=0.6628, tangentMode=RCTM_User, time=1.2, value=0.7}]}}}
      CurveIndex=$Emitter.Age
      Scale Curve=&FloatFromCurve(FloatCurve=DataInterface() {propertyValues {Curve {keys [
           {arriveTangent=4.1785, interpMode=RCIM_Cubic, leaveTangent=4.1785, tangentMode=RCTM_User, time=0, value=0.2}
           {arriveTangent=1.1179, interpMode=RCIM_Cubic, leaveTangent=1.1179, tangentMode=RCTM_User, time=0.2042, value=0.6228}
           {arriveTangent=0.0224, interpMode=RCIM_Cubic, leaveTangent=0.0224, tangentMode=RCTM_User, time=1, value=1.2}
          ]}}})
     )
    ))
   &ScaleColor(
    Scale Alpha=&FloatFromCurve(
     FloatCurve=DataInterface() {propertyValues {Curve {keys [
         <at 0 0.125 RCIM_Cubic>
         {arriveTangent=-2.5249, interpMode=RCIM_Cubic, leaveTangent=-2.5249, tangentMode=RCTM_User, time=0.3, value=1}
         {arriveTangent=-0.0405, interpMode=RCIM_Cubic, leaveTangent=-0.0405, tangentMode=RCTM_User, time=1, value=0}
        ]}}}
     Scale Curve=&FloatFromCurve(
      FloatCurve=DataInterface() {propertyValues {Curve {keys [
          {arriveTangent=1.0688, interpMode=RCIM_Cubic, leaveTangent=1.0688, tangentMode=RCTM_User, time=0, value=0.7}
          {arriveTangent=1.5842, tangentMode=RCTM_User, time=0.7, value=1}
         ]}}}
      CurveIndex=$Emitter.Age
     )
    )
    Scale RGB=<11 11 11>
    ScaleA=true
    ScaleRGB=true
   )
  ]
  rendererClasses [&NiagaraLightRendererProperties, &NiagaraMeshRendererProperties]
  renderers [
   {rendererClass=&NiagaraLightRendererProperties, rendererIndex=0}
   {rendererClass=&NiagaraMeshRendererProperties, rendererIndex=1}
  ]
 }
 MedLow {
  emitterUpdateScript [
   &EmitterState(
    Delay First Loop Only=true
    Inactive Response=::NewEnumerator1
    Life Cycle Mode=::NewEnumerator1
    Loop Delay=0.5
    Loop Duration=1
    UseLoopDelay=true
   )
   &SpawnRate(
    SpawnRate=&RandomRangeFloat(Minimum=-3, Maximum=6, Evaluation Type=::NewEnumerator1)
    Use Spawn Probability=false
   )
   &SpawnBurst_Instantaneous(Spawn Count=3, Use Spawn Probability=false, UseLoopCountLimit=false) {enabled=false}
  ]
  particleSpawnScript [
   &InitializeParticle(
    Color=<rgba 0 0.2943 1 1>
    Lifetime Max=0.6
    Lifetime Min=0.5
    Lifetime Mode=::NewEnumerator1
    Mesh Scale Mode=::NewEnumerator3
    Mesh Uniform Scale=&Multiply_Float(A=&RandomRangeFloat(Minimum=0.18, Maximum=0.21), B=0.25)
    Position=$Engine.Owner.Position
    Position Mode=::NewEnumerator1
    Sprite Size Max=<70 50>
    Sprite Size Min=<65 45>
    Sprite Size Mode=::NewEnumerator2
    UsePositionOffset=false
    Write Material Random=true
   ) {module=InitializeParticle001}
   &ShapeLocation(
    Cylinder Height=50
    Cylinder Radius=5
    MaskSpawnGroup=false
    Non Uniform Scale=<1 1.2 1>
    Offset=<0 0 20>
    Offset Coordinate Space=::Local
    Offset Mode=::NewEnumerator0
    Override Seed=false
    Shape Primitive=::NewEnumerator1
    Surface Only=false
    UseOwnerScale=true
   )
   /SetVariables(Particles.SpriteMesh_Quat:Quat4f=&RandomQuaternion(), Particles.Sprite_CamOfs=30) {isSetParameters=true, module=SetVariables_#3}
   &InitialMeshOrientation(
    Mesh Orientation Mode=::NewEnumerator1
    Rotation=&RandomRangeVector(Minimum=<-0.1 -0.1 -0.1>, Maximum=<0.1 0.1 0.1>)
    Use Rotation Vector=true
   )
   /SetVariables(
    Particles.InitPos:NiagaraPosition=$Particles.Initial.Position
    Particles.StoredVector:Vector3f=&TransformVector(
     Vector=Linked() {linkedVariable {name=Particles.Initial.Position, type=NiagaraPosition}}
     Source Space=::Simulation
     Destination Space=::World
     Apply Scale=true
     Transform Vector as Position=true
    )
   ) {enabled=false, isSetParameters=true, module=SetVariables_#4}
   &AddVelocity(
    Module.Use Falloff Distance=false
    Origin Offset=&RandomRangeVector(Minimum=<300 0 5>, Maximum=<500 -800 5>)
    Override Seed=false
    Velocity Mode=::NewEnumerator1
    Velocity Speed=&RandomRangeFloat(Minimum=100, Maximum=2500)
   ) {enabled=false}
   &AddVelocity(
    Velocity=&Multiply_Vector(
     A=&RandomRangeVector(Minimum=<0.6 0.6 0.6>, Maximum=<0.9 0.9 1>)
     B=&SelectVectorFromArray(
      Array Sampling Mode=::NewEnumerator0
      Direct Array Index=?
      Vector Selection Array=DataInterface() {propertyValues {FloatData [<-100 -100 -50>, <100 100 -60>]}}
      Random Seed=0
     )
    )
    Velocity Speed Scale=20
   ) {module=AddVelocity001}
  ]
  particleUpdateScript [
   &ParticleState()
   &SkeletalMeshLocation(
    Bone Sampling Mode=::NewEnumerator4
    MaskSpawnGroup=false
    Orientation Sampling=::NewEnumerator1
    Position Sampling=::NewEnumerator1
    Sampled Position Offset=<20 0 -77>
    Skeletal Mesh=DataInterface() {propertyValues {
      DefaultMesh=null
      ExcludeBoneName=root
      OptionalComponentClass=null
      PreviewMesh=&SK_M_MALE_Base_Skeleton
      SkinningMode=SkinOnTheFly
      SoftSourceActor=null
      SourceMode=Default
      UvSetIndex=0
      WholeMeshLOD=-1
      bExcludeBone=false
      bReadDeformedGeometry=true
      bRequireCurrentFrameData=true
      ComponentTags []
      FilteredBones [hand_l]
      FilteredSockets [FX_AnimTrail_01, FX_AnimTrail_02]
      MeshUserParameter {parameter {name=None, typeDefHandle {registeredTypeIndex=1}, varData []}}
      SamplingRegions []
     }}
    Source Movement Scale:Vector3f=&VectorFromCurve(VectorCurve=DataInterface() {propertyValues {XCurve {keys [<at 0.01 1>,<at 0.1 0.1>]}, YCurve {keys [<at 0.01 1>,<at 0.1 0.1>]}, ZCurve {keys [<at 0.01 1>,<at 0.1 0.1>]}}})
    Use Sampled Mesh Attribute as Ribbon ID=false
   )
   &UpdateMeshOrientation(Rotation Rate=&RandomRangeFloat(Minimum=-1, Maximum=1), Rotation Vector=<0.15 0.15 0.1>)
   &Drag(
    Drag=&FloatFromCurve(
     FloatCurve=DataInterface() {propertyValues {Curve {keys [{tangentMode=RCTM_User, time=0, value=0}, {tangentMode=RCTM_User, time=0.4, value=1}]}}}
     Scale Curve=90
    )
    Use Linear Drag=true
    Use Rotational Drag=true
   ) {enabled=false}
   &GravityForce(
    Coordinate Space=::Local
    Gravity=&Multiply_VectorByFloat(
     Vector=&AddVectorToPosition(Position=&RandomRangeVector(Minimum=<0 0 -35000>, Maximum=<0 0 -45000>), Vector=<0 0 0>)
     Float=&MakeCustomFloatFromBool(
      Bool=&SetBoolByFloatComparison(A=1, B=HlslExpression() {hlslExpression="length(Particles.Velocity)"})
      True Float=0
      False Float=1
     )
    )
   )
   &Collision(
    Age Colliding Particles=false
    CPU Collision Trace Channel=ECC_WorldStatic
    CPU Friction Blending:ENiagaraFrictionMergeType=::NewEnumerator3
    Collision Enabled=&SetBoolByFloatComparison(
     A=1
     B=&FloatFromCurve(FloatCurve=DataInterface() {propertyValues {Curve {keys [<at 0.05 0.95>,<at 0.5 2>]}}}, Scale Curve=0.5)
    )
    Enable Rest State=false
    Friction=10000
    Friction During a Bounce=10000
    KillOnCollision=false
    Method for Calculating Particle Radius:ENiagaraCalculateRadiusOptions=::NewEnumerator2
    Module.EnableMaxCollisionCount=false
    Module.Randomize Collision Normal Vector=false
    Randomize Collision Normal=0
    Restitution=0
   )
   &SolveForcesAndVelocity(Clamp Velocity=false, Limit Acceleration=false)
   &ScaleSpriteSize(
    Uniform Curve Scale=0.8
    Uniform Curve Sprite Scale=DataInterface() {propertyValues {Curve {keys [{tangentMode=RCTM_User, time=0, value=0.2}, <at 0.5 1>, {tangentMode=RCTM_User, time=0.9, value=0}]}}}
   )
   &ScaleMeshSize(Scale Factor=&Multiply_VectorByFloat(
     Vector=&MakeVector(
      X=1
      Y=1
      Z=&ClampFloat(Float=&Multiply_Float(A=10, B=HlslExpression() {hlslExpression="length(Particles.Velocity)"}), Min=5, Max=8)
     )
     Float=&MakeCustomFloatFromBool(
      Bool=&SetBoolByFloatComparison(A=10, B=HlslExpression() {hlslExpression="length(Particles.Velocity)"})
      True Float=&FloatFromCurve(
       FloatCurve=DataInterface() {propertyValues {Curve {keys [
           {tangentMode=RCTM_User, time=0, value=0.4}
           <at 0.1 0.5>
           <at 0.1651 0.3>
           <at 0.2226 0.45>
           {tangentMode=RCTM_User, time=0.3, value=0.25}
           <at 0.3288 0.45>
           <at 0.3636 0.4>
           <at 0.4 0.7>
           <at 0.4246 0.4>
           <at 0.4803 0.6>
           <at 0.5186 0.25>
           <at 0.5761 0.4>
           <at 0.6108 0.15>
           <at 0.6533 0.25>
           <at 0.7 0.4>
           <at 0.7391 0.183>
           <at 0.8 0.45>
           <at 0.8424 0.1684>
           <at 0.8845 0.3723>
           <at 0.9185 0.0999>
           <at 0.9483 0.4763>
           <at 0.9912 0.5>
          ]}}}
       Scale Curve=1.5
      )
      False Float=0
     )
    ))
   &DynamicMaterialParameters(
    Index 0 Param 1=HlslExpression() {hlslExpression=Particles.SkeletalMeshLocation.SampledPosition.x}
    Index 0 Param 2=HlslExpression() {hlslExpression=Particles.SkeletalMeshLocation.SampledPosition.y}
    Index 0 Param 3=HlslExpression() {hlslExpression=Particles.SkeletalMeshLocation.SampledPosition.z}
    Index 0 Param 4=11
    Index 1 Param 1=&FloatFromCurve(FloatCurve=DataInterface() {propertyValues {Curve {keys [<at 0.05 1 RCIM_Cubic>, {arriveTangent=-1.1493, tangentMode=RCTM_User, time=0.5, value=0}]}}})
    Module.Param0WriteEnabled=true
    Module.Param0WriteEnabled1=true
    Module.Param1WriteEnabled=true
    Module.Param1WriteEnabled1=false
    Module.Param2WriteEnabled=true
    Module.Param2WriteEnabled1=false
    Module.Param3WriteEnabled=true
    Module.Param3WriteEnabled1=false
    Write Parameter Index 1=true
   ) {module=DynamicMaterialParameters001}
   &ScaleColor(
    Scale Alpha=&MakeCustomFloatFromBool(
     Bool=&SetBoolByFloatComparison(A=5, B=HlslExpression() {hlslExpression="length(Particles.Velocity)"})
     True Float=&FloatFromCurve(
      FloatCurve=DataInterface() {propertyValues {Curve {keys [
          {arriveTangent=0.0978, interpMode=RCIM_Cubic, leaveTangent=0.0978, tangentMode=RCTM_User, time=0.1, value=0.65}
          {arriveTangent=0.0598, interpMode=RCIM_Cubic, leaveTangent=0.0598, tangentMode=RCTM_User, time=0.3, value=1.2}
          {arriveTangent=-2.4139, interpMode=RCIM_Cubic, leaveTangent=-2.4139, tangentMode=RCTM_User, time=0.9, value=0}
         ]}}}
      Scale Curve=&FNTimeOfDayScalar(Night Scale=0.6, Morning Scale=0.7, Evening Scale=0.7)
     )
     False Float=0
    )
    Scale RGB=<2 2 2>
    ScaleA=true
    ScaleRGB=true
   )
   /SetVariables(
    Particles.Color2=&Multiply_LinearColorByFloat(
     Linear Color=&ColorFromCurve(ColorCurve=DataInterface() {propertyValues {
        AlphaCurve {keys [<at 0.1016 0>, <at 0.1666 1>, <at 0.592 1 RCIM_Cubic>, {interpMode=RCIM_Cubic, leaveTangent=2.0391, time=0.76, value=0}]}
        BlueCurve {keys [<at 0.1152 0>,<at 0.2208 3 RCIM_Cubic>,<at 0.6137 3>,<at 0.7844 0>]}
        GreenCurve {keys [<at 0.1152 0>,<at 0.2208 0.8252 RCIM_Cubic>,<at 0.6137 0.8252>,<at 0.7844 0>]}
        RedCurve {keys [<at 0.1152 0>,<at 0.2208 0 RCIM_Cubic>,<at 0.6137 0>,<at 0.7844 0>]}
       }})
     Float=&MakeCustomFloatFromBool(
      Bool=&SetBoolByFloatComparison(A=10, B=HlslExpression() {hlslExpression="length(Particles.Velocity)"})
      True Float=1
      False Float=0
     )
    )
    Particles.Color3:LinearColor=&ColorFromCurve(ColorCurve=DataInterface() {propertyValues {
       AlphaCurve {keys [<at 0.0514 15 RCIM_Cubic>,<at 0.8816 0 RCIM_Cubic>]}
       BlueCurve {keys [<at 0 200 RCIM_Cubic>]}
       GreenCurve {keys [<at 0 95.4552 RCIM_Cubic>]}
       RedCurve {keys [<at 0 0 RCIM_Cubic>]}
      }})
    Particles.DM2_0:Vector4f=&MakeVector4(
     X=&Multiply_Float(
      A=&FloatFromCurve(
       FloatCurve=DataInterface() {propertyValues {Curve {keys [{tangentMode=RCTM_User, time=0, value=1}, {tangentMode=RCTM_User, time=0.8, value=0}]}}}
       Scale Curve=4
      )
      B=&MakeCustomFloatFromBool(
       Bool=&SetBoolByFloatComparison(A=5, B=HlslExpression() {hlslExpression="length(Particles.Velocity)"})
       True Float=0.2
       False Float=0
      )
     )
     Y=&FloatFromCurve(
      FloatCurve=DataInterface() {propertyValues {Curve {keys [
          {arriveTangent=2.2144, interpMode=RCIM_Cubic, leaveTangent=2.2144, tangentMode=RCTM_User, time=0, value=0}
          {tangentMode=RCTM_User, time=0.9, value=1}
         ]}}}
      Scale Curve=4
     )
     Z=&FloatFromCurve(
      FloatCurve=DataInterface() {propertyValues {Curve {keys [
          {arriveTangent=3.2315, interpMode=RCIM_Cubic, leaveTangent=3.2315, tangentMode=RCTM_User, time=0, value=0}
          {arriveTangent=1.1179, interpMode=RCIM_Cubic, leaveTangent=1.1179, tangentMode=RCTM_User, time=0.4, value=1}
         ]}}}
      Scale Curve=45
     )
     W=&FloatFromCurve(
      FloatCurve=DataInterface() {propertyValues {Curve {keys [
          {arriveTangent=0.6176, interpMode=RCIM_Cubic, leaveTangent=0.6176, tangentMode=RCTM_User, time=0.2, value=0}
          {arriveTangent=6.707, interpMode=RCIM_Cubic, leaveTangent=6.707, tangentMode=RCTM_User, time=1, value=2.5}
         ]}}}
      Scale Curve=-250
     )
    )
    Particles.Mesh Scale2:Vector3f=<10 10 10>
   ) {isSetParameters=true, module=SetVariables_#5}
   /SetVariables(Particles.Lifetime=&MakeCustomFloatFromBool(
     Bool=&SetBoolByFloatComparison(A=800, B=&DistanceBetweenPositions(A=$Engine.Owner.Position, B=$Particles.Position))
     True Float=$Particles.Lifetime
     False Float=0
    )) {isSetParameters=true, module=SetVariables_#6}
  ]
  rendererClasses [&NiagaraMeshRendererProperties, &NiagaraSpriteRendererProperties]
  renderers [
   {rendererClass=&NiagaraMeshRendererProperties, rendererIndex=0}
   {rendererClass=&NiagaraSpriteRendererProperties, rendererIndex=1}
   {rendererClass=&NiagaraMeshRendererProperties, rendererIndex=2}
  ]
 }
 Ring {rendererClasses [], renderers []}
 WindCone001 {rendererClasses [], renderers []}
}
summary {
 systemName=NS_ElectroBat_Charge2
 emitters [
  {emitterName=Flare, rendererClasses []}
  {emitterName=WindCone001, rendererClasses []}
  {emitterName=Lit, rendererClasses [&NiagaraLightRendererProperties, &NiagaraMeshRendererProperties]}
  {
   emitterName=Epic_High
   simTarget=GPUComputeSim
   rendererClasses [&NiagaraMeshRendererProperties, &NiagaraSpriteRendererProperties]
  }
  {emitterName=MedLow, rendererClasses [&NiagaraMeshRendererProperties, &NiagaraSpriteRendererProperties]}
  {emitterName=Flare002, rendererClasses []}
  {emitterName=LW_Lightning001, rendererClasses []}
  {emitterName=Lightning, rendererClasses [&NiagaraMeshRendererProperties]}
  {emitterName=Flare003, rendererClasses []}
  {emitterName=Ring, rendererClasses []}
  {emitterName=Dark, rendererClasses []}
 ]
 userVariables [User.ChargeUpFinishTime {type=NiagaraFloat, defaultValue=1, description=""}]
}
