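# NS_ElectroBat_Equip is a Niagara system for the ElectroBat gameplay feature, named as the weapon's equip effect.
# It holds five emitters named Sparks, Lightning, Flare, Mesh and LW_GlowSphereMesh1, with no renderer details recorded.
# User parameters expose a skeletal mesh data interface and an integer ribbon ID offset.
# The recorded modules are only SystemState, with a random float range dynamic input.
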NiagaraSystem /ElectroBatGameplay/Assets/VFX/NS_ElectroBat_Equip
dependencies {
 usedDataInterfaces []
 usedDynamicInputs [&RandomRangeFloat]
 usedModules [&SystemState]
}
emitters {
 Flare {rendererClasses [], renderers []}
 LW_GlowSphereMesh1 {rendererClasses [], renderers []}
 Lightning {rendererClasses [], renderers []}
 Mesh {rendererClasses [], renderers []}
 Sparks {rendererClasses [], renderers []}
}
summary {
 systemName=NS_ElectroBat_Equip
 emitters [
  {emitterName=Sparks, rendererClasses []}
  {emitterName=LW_GlowSphereMesh1, rendererClasses []}
  {emitterName=Flare, rendererClasses []}
  {emitterName=Mesh, rendererClasses []}
  {emitterName=Lightning, rendererClasses []}
 ]
 userVariables [
  User.RibbonIDOffset {type=NiagaraInt32}
  User.SkeletalMesh {
   type=NiagaraDataInterfaceSkeletalMesh
   defaultValue {
    struct=NiagaraExt_VariableValue_DataInterface
    value {dataInterface=&self:NiagaraDataInterfaceSkeletalMesh_0}
   }
  }
 ]
}
