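# Niagara system for the ElectroBat heavy swing "Range" effect, with two swing-trail emitters named as a mesh and a mesh glow.
# User parameters expose charge-up finish time, a blue main color and a glow color scale.
# System scripts use SystemState and a Set Variables module, with dynamic inputs for array selection, float math and vector/color conversion.
# Emitter internals and renderers are not listed; the LW_ prefix suggests lightweight emitters.
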NiagaraSystem /ElectroBatGameplay/Assets/VFX/NS_ElectroBat_HeavySwing_Range
dependencies {
 usedDataInterfaces [&NiagaraDataInterfaceArrayFloat, &NiagaraDataInterfaceArrayFloat3]
 usedDynamicInputs [
  &MakeVector
  &SelectFloatFromArray
  &MakeIntFromFloat
  &SelectVectorFromArray
  &MakeLinearColorFromVectorAndFloat
  &MakeVectorFromLinearColorRGB
  &Add_Float
  &Multiply_Float
 ]
 usedModules [
  &self:SystemSpawnScript.SystemScriptSource.SystemScriptGraph.NiagaraNodeAssignment_0.SetVariables_#1
  &SystemState
 ]
}
emitters {LW_SwingTrail_Mesh {rendererClasses [], renderers []}, LW_SwingTrail_MeshGlow {rendererClasses [], renderers []}}
summary {
 systemName=NS_ElectroBat_HeavySwing_Range
 emitters [{emitterName=LW_SwingTrail_MeshGlow, rendererClasses []}, {emitterName=LW_SwingTrail_Mesh, rendererClasses []}]
 userVariables [
  User.ChargeUpFinishTime {type=NiagaraFloat, defaultValue=0, description=""}
  User.Color {type=LinearColor, description="", defaultValue {struct=LinearColor, value=<rgba 0 1.3333 5 0.5>}}
  User.GlowColorScale {type=LinearColor, description="", defaultValue {struct=LinearColor, value=<rgba 0 0.1396 2 15>}}
 ]
}
