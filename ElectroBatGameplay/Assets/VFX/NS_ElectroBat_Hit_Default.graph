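# Niagara system for the default hit effect of the ElectroBat gameplay feature.
# It combines four emitters named for a flash, streaks, sparks and dark chunks, suggesting a short electric impact burst.
# System logic uses only SystemState and a Set Variables assignment, with no user parameters or data interfaces exposed.
# The file lists no renderer or module detail for the emitters.
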NiagaraSystem /ElectroBatGameplay/Assets/VFX/NS_ElectroBat_Hit_Default
dependencies {
 usedDataInterfaces []
 usedDynamicInputs []
 usedModules [
  &self:SystemSpawnScript.SystemScriptSource.SystemScriptGraph.NiagaraNodeAssignment_0.SetVariables_#1
  &SystemState
 ]
}
emitters {
 Dark_Chunks {rendererClasses [], renderers []}
 LW_Flash {rendererClasses [], renderers []}
 LW_Sparks1 {rendererClasses [], renderers []}
 LW_Streaks {rendererClasses [], renderers []}
}
summary {
 systemName=NS_ElectroBat_Hit_Default
 emitters [
  {emitterName=LW_Flash, rendererClasses []}
  {emitterName=LW_Streaks, rendererClasses []}
  {emitterName=Dark_Chunks, rendererClasses []}
  {emitterName=LW_Sparks1, rendererClasses []}
 ]
 userVariables []
}
