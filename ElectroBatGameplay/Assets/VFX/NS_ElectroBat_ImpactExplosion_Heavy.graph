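# Niagara system for the ElectroBat's heavy impact explosion, a short blue electric blast.
# GroundWave, OuterSphere and Wind meshes expand and fade through curve-driven scale, color and dynamic material parameters, with lights from Lit and OuterSphere.
# Empty002 adds rising ribbons in four spawn groups, spawned by distance moved and seeded on skeletal mesh bones.
# LW_Lightning, LW_Flash and Dark_Chunks list no modules or renderers.
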NiagaraSystem /ElectroBatGameplay/Assets/VFX/NS_ElectroBat_ImpactExplosion_Heavy
dependencies {
 usedDataInterfaces [
  &NiagaraDataInterfaceCurve
  &NiagaraDataInterfaceColorCurve
  &NiagaraDataInterfaceArrayFloat
  &NiagaraDataInterfaceSkeletalMesh
  &NiagaraDataInterfaceArrayInt32
  &NiagaraDataInterfaceVectorCurve
 ]
 usedDynamicInputs []
 usedModules [
  &self:SystemSpawnScript.SystemScriptSource.SystemScriptGraph.NiagaraNodeAssignment_0.SetVariables_#1
  &SystemState
 ]
}
emitters {
 Dark_Chunks {rendererClasses [], renderers []}
 Empty002 {
  emitterUpdateScript [
   &EmitterState()
   &SpawnBurst_Instantaneous(Use Spawn Probability=false, UseLoopCountLimit=false) {enabled=false}
   &SpawnBurst_Instantaneous(Spawn Group=1, Use Spawn Probability=false, UseLoopCountLimit=false) {enabled=false, module=SpawnBurst_Instantaneous001}
   &SpawnBurst_Instantaneous(Spawn Group=2, Use Spawn Probability=false, UseLoopCountLimit=false) {enabled=false, module=SpawnBurst_Instantaneous002}
   &SpawnBurst_Instantaneous(Spawn Group=3, Use Spawn Probability=false, UseLoopCountLimit=false) {enabled=false, module=SpawnBurst_Instantaneous003}
   &SpawnPerUnit(Module.Use Max Movement Threshold=true, Module.Use Movement Tolerance=false, Spawn Spacing=50, Use Spawn Probability=false)
   &SpawnPerUnit(Module.Use Max Movement Threshold=true, Module.Use Movement Tolerance=false, Spawn Group=1, Spawn Spacing=50, Use Spawn Probability=false) {module=SpawnPerUnit001}
   &SpawnPerUnit(Module.Use Max Movement Threshold=true, Module.Use Movement Tolerance=false, Spawn Group=2, Spawn Spacing=50, Use Spawn Probability=false) {module=SpawnPerUnit002}
   &SpawnPerUnit(Module.Use Max Movement Threshold=true, Module.Use Movement Tolerance=false, Spawn Group=3, Spawn Spacing=50, Use Spawn Probability=false) {module=SpawnPerUnit003}
  ]
  particleSpawnScript [
   &InitializeParticle(
    Color Mode=::NewEnumerator0
    Lifetime=&SelectFloatFromArray(Direct Array Index=$Emitter.SpawnGroup, Float Selection Array=DataInterface() {propertyValues {FloatData [0.3,0.25,0.2,0.28]}})
    Ribbon Width=&SelectFloatFromArray(Direct Array Index=$Emitter.SpawnGroup, Float Selection Array=DataInterface() {propertyValues {FloatData [30,30,40,40]}})
    Ribbon Width Mode=::NewEnumerator1
    UsePositionOffset=false
    Write Material Random=true
   )
   &SkeletalMeshLocation(
    Bone / Socket Index=&SelectIntFromArray(Direct Array Index=$Emitter.SpawnGroup, Int Selection Array=DataInterface() {propertyValues {IntData [0,7,12,18]}})
    Bone Sampling Mode=::NewEnumerator4
    MaskSpawnGroup=false
    Orientation Sampling=::NewEnumerator1
    Skeletal Mesh=DataInterface() {propertyValues {
      DefaultMesh=&M_Med_Soldier_04_Preview
      ExcludeBoneName=root
      OptionalComponentClass=null
      PreviewMesh=&M_Med_Soldier_04_Preview
      SkinningMode=SkinOnTheFly
      SoftSourceActor=null
      SourceMode=Default
      UvSetIndex=0
      WholeMeshLOD=-1
      bExcludeBone=false
      bReadDeformedGeometry=true
      bRequireCurrentFrameData=true
      ComponentTags []
      FilteredBones [
       upperarm_l
       lowerarm_l
       hand_l
       upperarm_twist_02_l
       lowerarm_twist_02_l
       lowerarm_twist_01_l
       upperarm_r
       lowerarm_r
       hand_r
       upperarm_twist_02_r
       lowerarm_twist_01_r
       lowerarm_twist_02_r
       thigh_l
       thigh_twist_01_l
       calf_l
       calf_twist_02_l
       thigh_r
       thigh_twist_01_r
       calf_r
       calf_twist_02_r
       pelvis
       head
      ]
      FilteredSockets []
      MeshUserParameter {parameter {name=None, typeDefHandle {registeredTypeIndex=1}, varData []}}
      SamplingRegions []
     }}
   )
  ]
  particleUpdateScript [
   &ParticleState()
   &DynamicMaterialParameters(
    Index 1 Param 1=&SelectFloatFromArray(Direct Array Index=$Emitter.SpawnGroup, Float Selection Array=DataInterface() {propertyValues {FloatData [0,1333.7534,22.5788,140.1219]}})
    Index 1 Param 2=1
    Index 1 Param 3=&FloatFromCurve(FloatCurve=DataInterface() {propertyValues {Curve {keys [<at 0 0>,<at 1 1>]}}}, Scale Curve=3)
    Index 1 Param 4=-1
    Module.Index1Param0Write=true
    Module.Index1Param1Write=true
    Module.Index1Param2Write=true
    Module.Index1Param3Write=true
    Module.Index2Param0Write=true
    Module.Index2Param1Write=true
    Module.Index2Param2Write=true
    Module.Index2Param3Write=true
    Module.Index3Param0Write=true
    Module.Index3Param1Write=true
    Module.Index3Param2Write=true
    Module.Index3Param3Write=true
    Write Parameter Index 0=false
    Write Parameter Index 1=true
    Write Parameter Index 2=true
    Write Parameter Index 3=true
   )
   &ScaleRibbonWidth(Ribbon Width Scale=&FloatFromCurve(
     FloatCurve=DataInterface() {propertyValues {Curve {keys [
         <at 0 0 RCIM_Cubic>
         {interpMode=RCIM_Cubic, leaveTangent=-1.5853, tangentMode=RCTM_Break, time=0.2, value=1}
         {arriveTangent=-0.0717, interpMode=RCIM_Cubic, tangentMode=RCTM_Break, time=1, value=0}
        ]}}}
     Scale Curve=2
    ))
   /SetVariables(Particles.RibbonWidth2:NiagaraFloat=&FloatFromCurve(
     FloatCurve=DataInterface() {propertyValues {Curve {keys [
         {interpMode=RCIM_Cubic, leaveTangent=1.4784, tangentMode=RCTM_Break, time=0, value=1.275}
         {arriveTangent=-1.4054, interpMode=RCIM_Cubic, leaveTangent=-1.4054, tangentMode=RCTM_User, time=0.2, value=1.375}
         {arriveTangent=-1.5652, interpMode=RCIM_Cubic, tangentMode=RCTM_Break, time=1, value=0}
        ]}}}
     Scale Curve=50
    )) {isSetParameters=true}
   &ScaleColor(
    Scale Alpha=&FloatFromCurve(
     FloatCurve=DataInterface() {propertyValues {Curve {keys [<at 0 0>,<at 0.1 0.9>,<at 1 1>]}}}
     Scale Curve=&FloatFromCurve(FloatCurve=DataInterface() {propertyValues {Curve {keys [<at 0.4 1>,<at 0.8 0>]}}}, CurveIndex=$Emitter.Age, Scale Curve=0.5)
    )
    Scale RGB=<0.5 1 2>
    ScaleA=true
    ScaleRGB=true
   )
   &RotateAroundPoint(Radius=25, Rotation Center:NiagaraPosition=$Particles.SkeletalMeshLocation.SampledPosition, Rotation Phase=5, Rotation Rate=5) {enabled=false}
   &InheritSourceMovement(Applied Position Scale=<0.7 0.7 0.7>, ApplyForce=false, ApplyPosition=true, ApplyPositionFalloff=false) {enabled=false}
   &AccelerationForce(Acceleration=<0 0 500>)
   &SolveForcesAndVelocity(Clamp Velocity=false, Limit Acceleration=false)
  ]
  rendererClasses [&NiagaraRibbonRendererProperties]
  renderers [{rendererClass=&NiagaraRibbonRendererProperties}]
 }
 GroundWave {
  emitterUpdateScript [
   &EmitterState()
   &SpawnBurst_Instantaneous(Use Spawn Probability=false, UseLoopCountLimit=false)
  ]
  particleSpawnScript [
   &InitializeParticle(
    Color=<rgba 1.25 1.25 1.25 1>
    Lifetime=&MaskFloatBySpawnGroup(Masked Float=0.4, Passthrough Float=0.4)
    Mesh Scale Mode=::NewEnumerator3
    Mesh Uniform Scale=&MaskFloatBySpawnGroup(Masked Float=2.5, Passthrough Float=3.5)
    UsePositionOffset=false
    Write Material Random=true
   )
   &InitialMeshOrientation(
    Mesh Orientation Mode=::NewEnumerator2
    Orientation Vector=<0 0 1>
    Rotation=&RandomRangeVector(Minimum=<0 0 0>, Maximum=<0 0 1>)
    Use Rotation Vector=true
   )
   &CameraOffset(Camera Offset Amount=40) {enabled=false}
  ]
  particleUpdateScript [
   &ParticleState()
   &ScaleMeshSize(Scale Factor=&VectorFromCurve(VectorCurve=DataInterface() {propertyValues {
       XCurve {keys [
         {arriveTangent=1.4912, interpMode=RCIM_Cubic, leaveTangent=1.4912, tangentMode=RCTM_User, time=0, value=0.2}
         {arriveTangent=0.1443, interpMode=RCIM_Cubic, leaveTangent=0.1443, tangentMode=RCTM_User, time=1, value=1}
        ]}
       YCurve {keys [
         {arriveTangent=1.4912, interpMode=RCIM_Cubic, leaveTangent=1.4912, tangentMode=RCTM_User, time=0, value=0.2}
         {arriveTangent=0.1443, interpMode=RCIM_Cubic, leaveTangent=0.1443, tangentMode=RCTM_User, time=1, value=1}
        ]}
       ZCurve {keys [
         {arriveTangent=5.9564, interpMode=RCIM_Cubic, leaveTangent=5.9564, tangentMode=RCTM_User, time=0, value=0.5}
         {arriveTangent=-0.1, interpMode=RCIM_Cubic, leaveTangent=-0.1, time=0.2, value=1}
         {arriveTangent=-1.1924, interpMode=RCIM_Cubic, leaveTangent=-1.1924, tangentMode=RCTM_User, time=1, value=0.4}
        ]}
      }})) {enabled=false}
   &ScaleMeshSize(Scale Factor=&MaskVectorBySpawnGroup(
     Masked Vector=&VectorFromCurve(VectorCurve=DataInterface() {propertyValues {
        XCurve {keys [
          {arriveTangent=2.2781, interpMode=RCIM_Cubic, leaveTangent=2.2781, tangentMode=RCTM_User, time=0, value=0.1}
          {arriveTangent=0.288, interpMode=RCIM_Cubic, leaveTangent=0.288, tangentMode=RCTM_User, time=1, value=1}
         ]}
        YCurve {keys [
          {arriveTangent=2.2781, interpMode=RCIM_Cubic, leaveTangent=2.2781, tangentMode=RCTM_User, time=0, value=0.1}
          {arriveTangent=0.288, interpMode=RCIM_Cubic, leaveTangent=0.288, tangentMode=RCTM_User, time=1, value=1}
         ]}
        ZCurve {keys [
          {arriveTangent=3.6292, interpMode=RCIM_Cubic, leaveTangent=3.6292, tangentMode=RCTM_User, time=0, value=0.5}
          {arriveTangent=-1.1338, interpMode=RCIM_Cubic, leaveTangent=-1.1338, tangentMode=RCTM_User, time=0.3, value=0.6}
          {arriveTangent=-0.1814, interpMode=RCIM_Cubic, leaveTangent=-0.1814, tangentMode=RCTM_User, time=1, value=0.4}
         ]}
       }})
     Passthrough Vector=&VectorFromCurve(
      VectorCurve=DataInterface() {propertyValues {
        XCurve {keys [
          {arriveTangent=3.9002, interpMode=RCIM_Cubic, leaveTangent=3.9002, tangentMode=RCTM_User, time=0, value=0.2}
          {arriveTangent=1.2534, interpMode=RCIM_Cubic, leaveTangent=1.2534, tangentMode=RCTM_User, time=0.1743, value=0.6343}
          {arriveTangent=0.2006, interpMode=RCIM_Cubic, leaveTangent=0.2006, tangentMode=RCTM_User, time=1, value=1}
         ]}
        YCurve {keys [
          {arriveTangent=3.9002, interpMode=RCIM_Cubic, leaveTangent=3.9002, tangentMode=RCTM_User, time=0, value=0.2}
          {arriveTangent=1.2534, interpMode=RCIM_Cubic, leaveTangent=1.2534, tangentMode=RCTM_User, time=0.1743, value=0.6343}
          {arriveTangent=0.2006, interpMode=RCIM_Cubic, leaveTangent=0.2006, tangentMode=RCTM_User, time=1, value=1}
         ]}
        ZCurve {keys [
          {arriveTangent=8.292, interpMode=RCIM_Cubic, leaveTangent=8.292, tangentMode=RCTM_User, time=0, value=0.16}
          {arriveTangent=-1.3261, interpMode=RCIM_Cubic, leaveTangent=-1.3261, tangentMode=RCTM_User, time=0.07, value=0.6}
          <at 1 0.2>
         ]}
       }}
      Scale Curve=<0.7 0.7 0.7>
     )
    )) {module=ScaleMeshSize001}
   &DynamicMaterialParameters(
    Index 0 Param 1=&MaskFloatBySpawnGroup(
     Masked Float=1
     Passthrough Float=&FloatFromCurve(FloatCurve=DataInterface() {propertyValues {Curve {keys [{arriveTangent=-3.0733, interpMode=RCIM_Cubic, leaveTangent=-3.0733, tangentMode=RCTM_User, time=0, value=1}, <at 1 -0.75 RCIM_Cubic>]}}})
    )
    Index 0 Param 2=&MaskFloatBySpawnGroup(
     Masked Float=1.75
     Passthrough Float=&FloatFromCurve(FloatCurve=DataInterface() {propertyValues {Curve {keys [
          {arriveTangent=-2.0952, interpMode=RCIM_Cubic, leaveTangent=-2.0952, tangentMode=RCTM_User, time=0, value=1.5}
          {arriveTangent=-1.4744, interpMode=RCIM_Cubic, leaveTangent=-1.4744, tangentMode=RCTM_User, time=1, value=0.25}
         ]}}})
    )
    Index 0 Param 3=&FloatFromCurve(
     FloatCurve=DataInterface() {propertyValues {Curve {keys [
         {arriveTangent=-2.0927, interpMode=RCIM_Cubic, leaveTangent=-2.0927, tangentMode=RCTM_User, time=0, value=1}
         {interpMode=RCIM_Cubic, leaveTangent=-1.7693, time=1, value=0}
        ]}}}
     Scale Curve=0.25
    )
    Index 0 Param 4=&FloatFromCurve(FloatCurve=DataInterface() {propertyValues {Curve {keys [{arriveTangent=1.0328, interpMode=RCIM_Cubic, leaveTangent=1.0328, tangentMode=RCTM_User, time=0, value=0}, <at 1 1 RCIM_Cubic>]}}})
    Module.Index0Float0Write=true
    Module.Index0Float1Write=true
    Module.Index0Float2Write=true
    Module.Index0Float3Write=true
   )
   &InheritSourceMovement(Applied Position Scale=<-0.5 -0.5 -0.5>, ApplyForce=false, ApplyPosition=true, ApplyPositionFalloff=false)
   &SolveForcesAndVelocity(Clamp Velocity=false, Limit Acceleration=false)
   &ScaleColor(
    Scale Alpha=&FloatFromCurve(FloatCurve=DataInterface() {propertyValues {Curve {keys [
         {arriveTangent=26.1682, interpMode=RCIM_Cubic, leaveTangent=26.1682, tangentMode=RCTM_User, time=0, value=0}
         {interpMode=RCIM_Cubic, tangentMode=RCTM_User, time=0.1, value=1}
         {tangentMode=RCTM_User, time=1, value=0}
        ]}}})
    Scale RGB=&Lerp_Vector(
     A=<1 2 3>
     B=<0.5 1.5 1.8>
     Alpha=&FloatFromCurve(
      FloatCurve=DataInterface() {propertyValues {Curve {keys [{interpMode=RCIM_Cubic, tangentMode=RCTM_User, time=0.1, value=0}, {interpMode=RCIM_Cubic, tangentMode=RCTM_User, time=0.4, value=1}]}}}
      CurveIndex=$Emitter.Age
     )
    )
    ScaleA=true
    ScaleRGB=true
   )
  ]
  rendererClasses [&NiagaraMeshRendererProperties]
  renderers [{rendererClass=&NiagaraMeshRendererProperties, rendererIndex=0}]
 }
 LW_Flash {rendererClasses [], renderers []}
 LW_Lightning {rendererClasses [], renderers []}
 Lit {
  emitterSpawnScript [/SetVariables(Emitter.CamOfs=50, Emitter.Position2:NiagaraPosition=<-6 0 2>) {isSetParameters=true}]
  emitterUpdateScript [
   &EmitterState(
    Inactive Response=::NewEnumerator1
    Life Cycle Mode=::NewEnumerator1
    Loop Behavior=::NewEnumerator1
    Loop Delay=0.9
    Loop Duration=0.9
    UseLoopDelay=false
   )
   &SpawnBurst_Instantaneous(Spawn Count=1, Spawn Probability=$User.Spawn Stage3, Use Spawn Probability=false)
   /SetVariables(
    Emitter.Decal Color:LinearColor=&ColorFromCurve(
     ColorCurve=DataInterface() {propertyValues {
       AlphaCurve {keys []}
       BlueCurve {keys [<at 0 1 RCIM_Cubic>,<at 0.4704 1>,<at 0.6923 0>]}
       GreenCurve {keys [<at 0 0.6843 RCIM_Cubic>,<at 0.4704 0.1589>,<at 0.6923 0>]}
       RedCurve {keys [<at 0 0 RCIM_Cubic>,<at 0.4704 0>,<at 0.6923 0>]}
      }}
     CurveIndex=&Frac_Float(Float=$Emitter.Age)
    )
    Emitter.DecalFade:NiagaraFloat=&FloatFromCurve(
     FloatCurve=DataInterface() {propertyValues {Curve {keys [<at 0 1 RCIM_Cubic>,<at 0.7 0 RCIM_Cubic>]}}}
     CurveIndex=$Emitter.Age
     Scale Curve=0.1
    )
    Emitter.DecalSize:Vector3f=&Multiply_VectorByFloat(Vector=<650 650 650>, Float=1)
    Emitter.Scale:Vector3f=&Multiply_VectorByFloat(
     Vector=&Multiply_VectorByFloat(Vector=&Multiply_VectorByFloat(Vector=<1 1 1>, Float=0.35), Float=1)
     Float=&Sine(
      Normalized Angle=$Emitter.Age
      Period=&Divide_Float(
       A=2
       B=&Add_Float(
        A=1
        B=&Add_Float(
         A=&Ceil_Float(Float=$User.Spawn Stage2)
         B=&Add_Float(
          A=&Multiply_Float(A=&Ceil_Float(Float=$User.Spawn Stage3), B=5)
          B=&Multiply_Float(A=&Ceil_Float(Float=$User.Spawn Stage4), B=6)
         )
        )
       )
      )
      Scale=0.1
      Bias=0.8
     )
    )
   ) {enabled=false, isSetParameters=true}
   /SetVariables(
    Emitter.LightColor:LinearColor=&ColorFromCurve(
     ColorCurve=DataInterface() {propertyValues {
       AlphaCurve {keys [<at 0.0069 0>,<at 0.0963 1 RCIM_Cubic>,<at 1 0 RCIM_Cubic>]}
       BlueCurve {keys [<at 0.0115 0>,<at 0.0871 70 RCIM_Cubic>,<at 1 0 RCIM_Cubic>]}
       GreenCurve {keys [<at 0.0115 0>,<at 0.0871 21.3686 RCIM_Cubic>,<at 1 0 RCIM_Cubic>]}
       RedCurve {keys [<at 0.0115 0>,<at 0.0871 5.8333 RCIM_Cubic>,<at 1 0 RCIM_Cubic>]}
      }}
     CurveIndex=$Emitter.NormalizedLoopAge
    )
    Emitter.LightRadius:NiagaraFloat=&FloatFromCurve(FloatCurve=DataInterface() {propertyValues {Curve {keys [<at 0 1 RCIM_Cubic>,<at 0.3 0>]}}}, CurveIndex=$Emitter.Age, Scale Curve=380)
    Emitter.MaskCenterGradientPow:NiagaraFloat=&FloatFromCurve(
     FloatCurve=DataInterface() {propertyValues {Curve {keys [{arriveTangent=-45.0315, interpMode=RCIM_Cubic, leaveTangent=-45.0315, tangentMode=RCTM_User, time=0, value=20}, <at 0.6 5>]}}}
     CurveIndex=$Emitter.Age
    )
   ) {isSetParameters=true}
  ]
  rendererClasses [&NiagaraLightRendererProperties]
  renderers [{rendererClass=&NiagaraLightRendererProperties, rendererIndex=0}]
 }
 OuterSphere {
  emitterUpdateScript [
   &EmitterState()
   &SpawnBurst_Instantaneous(Use Spawn Probability=false, UseLoopCountLimit=false)
  ]
  particleSpawnScript [
   &InitializeParticle(
    Color=<rgba 1.5 1.5 1.5 3>
    Lifetime=0.3
    Mesh Scale=<3.2 3.2 5>
    Mesh Scale Mode=::NewEnumerator4
    UsePositionOffset=false
    Write Material Random=true
   )
   &InitialMeshOrientation(
    Mesh Orientation Mode=::NewEnumerator2
    Orientation Vector=<0 0 1>
    Rotation=&RandomRangeVector(Minimum=<0 0 0>, Maximum=<0 0 1>)
    Use Rotation Vector=true
   )
   /SetVariables(Particles.LightPosition:NiagaraPosition=<0 0 40>) {isSetParameters=true}
  ]
  particleUpdateScript [
   &ParticleState()
   /SetVariables(Particles.Position=&MakeVector(
     X=0
     Y=0
     Z=&FloatFromCurve(
      FloatCurve=DataInterface() {propertyValues {Curve {keys [
          {arriveTangent=5.5536, interpMode=RCIM_Cubic, leaveTangent=5.5536, tangentMode=RCTM_User, time=0, value=0}
          {arriveTangent=1, interpMode=RCIM_Cubic, leaveTangent=1, time=0.2, value=0.7}
          {arriveTangent=0.1617, interpMode=RCIM_Cubic, leaveTangent=0.1617, tangentMode=RCTM_User, time=1, value=1}
         ]}}}
      Scale Curve=25
     )
    )) {isSetParameters=true}
   /SetVariables(Particles.LightColor=&ColorFromCurve(ColorCurve=DataInterface() {propertyValues {
       AlphaCurve {keys [
         <at 0 1 RCIM_Cubic>
         {arriveTangent=-2.4738, interpMode=RCIM_Cubic, leaveTangent=-2.4738, tangentMode=RCTM_User, time=0.3627, value=0.7008}
         <at 1 0 RCIM_Cubic>
        ]}
       BlueCurve {keys [<at 0.1733 36.9992 RCIM_Cubic>,<at 0.647 25 RCIM_Cubic>]}
       GreenCurve {keys [<at 0.1733 1.9736 RCIM_Cubic>,<at 0.647 10.9288 RCIM_Cubic>]}
       RedCurve {keys [<at 0.1733 40 RCIM_Cubic>,<at 0.647 4.126 RCIM_Cubic>]}
      }})) {isSetParameters=true}
   &ScaleMeshSize(Scale Factor=&VectorFromFloat(Value=&FloatFromCurve(
      FloatCurve=DataInterface() {propertyValues {Curve {keys [
          {arriveTangent=5.5536, interpMode=RCIM_Cubic, leaveTangent=5.5536, tangentMode=RCTM_User, time=0, value=0.2}
          {arriveTangent=0.8, interpMode=RCIM_Cubic, leaveTangent=0.8, time=0.2, value=0.8}
          {arriveTangent=0.1617, interpMode=RCIM_Cubic, leaveTangent=0.1617, tangentMode=RCTM_User, time=1, value=1}
         ]}}}
      Scale Curve=0.5
     )))
   &DynamicMaterialParameters(
    Index 0 Param 1=&FloatFromCurve(
     FloatCurve=DataInterface() {propertyValues {Curve {keys [
         {arriveTangent=-2.9934, interpMode=RCIM_Cubic, leaveTangent=-2.9934, tangentMode=RCTM_User, time=0, value=1}
         {interpMode=RCIM_Cubic, leaveTangent=-1.2287, time=1, value=-0.25}
        ]}}}
     Scale Curve=2
    )
    Index 0 Param 2=1
    Index 0 Param 3=&FloatFromCurve(
     FloatCurve=DataInterface() {propertyValues {Curve {keys [
         {arriveTangent=2.2611, interpMode=RCIM_Cubic, leaveTangent=2.2611, tangentMode=RCTM_User, time=0, value=0}
         {arriveTangent=0.402, interpMode=RCIM_Cubic, leaveTangent=0.402, tangentMode=RCTM_User, time=1, value=1}
        ]}}}
     Scale Curve=5
    )
    Index 0 Param 4=&FloatFromCurve(FloatCurve=DataInterface() {propertyValues {Curve {keys [{arriveTangent=-2.6143, interpMode=RCIM_Cubic, time=0, value=1}, <at 0.7 0>]}}})
    Module.Param0WriteEnabled=true
    Module.Param1WriteEnabled=true
    Module.Param2WriteEnabled=true
    Module.Param3WriteEnabled=true
   )
   &ScaleColor(
    Scale Alpha=&FloatFromCurve(
     FloatCurve=DataInterface() {propertyValues {Curve {keys [
         {arriveTangent=26.1682, interpMode=RCIM_Cubic, leaveTangent=26.1682, tangentMode=RCTM_User, time=0, value=0}
         {arriveTangent=-2.2763, interpMode=RCIM_Cubic, leaveTangent=-2.2763, tangentMode=RCTM_User, time=0.07, value=1}
         {tangentMode=RCTM_User, time=1, value=0}
        ]}}}
     Scale Curve=5
    )
    Scale RGB=&Lerp_Vector(
     A=<0.2 0.7 1.2>
     B=<0.2 1 1>
     Alpha=&FloatFromCurve(
      FloatCurve=DataInterface() {propertyValues {Curve {keys [{interpMode=RCIM_Cubic, tangentMode=RCTM_User, time=0.1, value=0}, {interpMode=RCIM_Cubic, tangentMode=RCTM_User, time=0.4, value=1}]}}}
      CurveIndex=$Emitter.Age
     )
    )
    ScaleA=true
    ScaleRGB=true
   )
   &InheritSourceMovement(Applied Position Scale=<-0.5 -0.5 -0.5>, ApplyForce=false, ApplyPosition=true, ApplyPositionFalloff=false)
   &SolveForcesAndVelocity(Clamp Velocity=false, Limit Acceleration=false)
  ]
  rendererClasses [&NiagaraMeshRendererProperties, &NiagaraLightRendererProperties]
  renderers [
   {rendererClass=&NiagaraMeshRendererProperties, rendererIndex=0}
   {rendererClass=&NiagaraLightRendererProperties, rendererIndex=1}
  ]
 }
 Wind {
  emitterUpdateScript [
   &EmitterState(Enable Distance Culling=true, MaxDistance=6000, Module.bUseMaxDistance=true, Module.bUseMinDistance=false, Scalability Mode=::NewEnumerator1)
   &SpawnRate(
    SpawnRate=&FloatFromCurve(
     FloatCurve=DataInterface() {propertyValues {Curve {keys [
         {arriveTangent=77.1005, interpMode=RCIM_Cubic, leaveTangent=77.1005, tangentMode=RCTM_User, time=0.05, value=0}
         <at 0.07 1 RCIM_Cubic>
         {arriveTangent=-14.3902, interpMode=RCIM_Cubic, leaveTangent=-14.3902, tangentMode=RCTM_User, time=0.25, value=0}
        ]}}}
     CurveIndex=$System.Age
     Scale Curve=40
    )
    Use Spawn Probability=false
   )
   &SpawnBurst_Instantaneous(Spawn Count=4, Spawn Time=0.05, Use Spawn Probability=false, UseLoopCountLimit=false)
  ]
  particleSpawnScript [
   &InitializeParticle(
    Color=<rgba 1 1 1 0.35>
    Lifetime=&UniformRangedFloat(Minimum=0.3, Maximum=0.47)
    Mesh Scale=&Lerp_Vector(A=<5 5 2.5>, B=<6.5 6.5 3.5>, Alpha=&RandomRangeFloat(Minimum=0, Maximum=1))
    Mesh Scale Mode=::NewEnumerator4
    UsePositionOffset=false
    Write Material Random=true
   ) {module=InitializeParticle001}
   &InitialMeshOrientation(
    Mesh Axis To Orient=<0 0 1>
    Mesh Orientation Mode=::NewEnumerator2
    Rotation=&RandomRangeVector(Minimum=<0 0 0>, Maximum=<0 0 1>)
    Use Rotation Vector=true
   )
  ]
  particleUpdateScript [
   &ParticleState()
   &ScaleMeshSize(Scale Factor=&VectorFromCurve(VectorCurve=DataInterface() {propertyValues {
       XCurve {keys [
         {arriveTangent=1.9177, interpMode=RCIM_Cubic, leaveTangent=1.9177, tangentMode=RCTM_User, time=0, value=0.1}
         {arriveTangent=0.2969, interpMode=RCIM_Cubic, leaveTangent=0.8713, tangentMode=RCTM_User, time=1, value=1}
        ]}
       YCurve {keys [
         {arriveTangent=1.9177, interpMode=RCIM_Cubic, leaveTangent=1.9177, tangentMode=RCTM_User, time=0, value=0.1}
         {arriveTangent=0.2969, interpMode=RCIM_Cubic, leaveTangent=0.5259, tangentMode=RCTM_User, time=1, value=1}
        ]}
       ZCurve {keys [
         {arriveTangent=2.8674, interpMode=RCIM_Cubic, leaveTangent=2.8674, tangentMode=RCTM_User, time=0, value=0.5}
         {arriveTangent=-0.4, interpMode=RCIM_Cubic, leaveTangent=-0.4, time=0.35, value=1}
         {arriveTangent=-1.6862, interpMode=RCIM_Cubic, leaveTangent=-1.6862, tangentMode=RCTM_User, time=1, value=0.1}
        ]}
      }}))
   /SetVariables(Particles.Position=&AddVectorToPosition(
     Position=&Multiply_VectorByFloat(
      Vector=&ReturnMeshOrientationAxis(Mesh Orientation Axis=::XAxis, Mesh Orientation Quat=$Particles.MeshOrientation)
      Float=&FloatFromCurve(
       FloatCurve=DataInterface() {propertyValues {Curve {keys [
           {arriveTangent=3.4662, interpMode=RCIM_Cubic, leaveTangent=3.4662, tangentMode=RCTM_User, time=0, value=-0.2}
           {interpMode=RCIM_Cubic, tangentMode=RCTM_User, time=0.5, value=1}
          ]}}}
       CurveIndex=$Emitter.Age
       Scale Curve=&Multiply_Float(A=&MakeFloatFromVector(Vector=$Particles.Initial.Scale, Channel=::NewEnumerator0), B=18)
      )
     )
     Vector=&MakeVector(X=0, Y=0, Z=&RandomRangeFloat(Minimum=25, Maximum=60))
    )) {isSetParameters=true}
   &CameraOffset(Camera Offset Amount=&Lerp_Float(A=70, B=140, Alpha=$Particles.NormalizedAge))
   &ScaleColor(
    Scale Alpha=&FloatFromCurve(FloatCurve=DataInterface() {propertyValues {Curve {keys [
         {arriveTangent=8.125, interpMode=RCIM_Cubic, time=0, value=0}
         {interpMode=RCIM_Cubic, tangentMode=RCTM_User, time=0.3, value=1}
         {arriveTangent=-0.8126, interpMode=RCIM_Cubic, leaveTangent=-0.8126, tangentMode=RCTM_User, time=0.7791, value=0.7}
         {interpMode=RCIM_Cubic, leaveTangent=-2.3395, time=1, value=0}
        ]}}})
    Scale RGB=&MakeVectorFromLinearColorRGB(Color=&ColorFromCurve(
      ColorCurve=DataInterface() {propertyValues {
        AlphaCurve {keys []}
        BlueCurve {keys [<at 0.0485 6 RCIM_Cubic>, {arriveTangent=-10.5416, interpMode=RCIM_Cubic, leaveTangent=-10.5416, time=0.1696, value=2}, <at 0.5564 0.6458>]}
        GreenCurve {keys [
          <at 0.0485 1.4895 RCIM_Cubic>
          {arriveTangent=-2.6943, interpMode=RCIM_Cubic, leaveTangent=-2.6943, time=0.1696, value=0.7785}
          <at 0.5564 0.1211>
         ]}
        RedCurve {keys [<at 0.0485 0 RCIM_Cubic>, {arriveTangent=0.1589, interpMode=RCIM_Cubic, leaveTangent=0.1589, time=0.1696, value=0.4433}, <at 0.5564 0.0807>]}
       }}
      CurveIndex=$Emitter.Age
     ))
    ScaleA=true
    ScaleRGB=true
   )
   &DynamicMaterialParameters(
    Index 0 Param 1=&FloatFromCurve(FloatCurve=DataInterface() {propertyValues {Curve {keys [
         {arriveTangent=-13.4039, interpMode=RCIM_Cubic, leaveTangent=-13.4039, tangentMode=RCTM_User, time=0, value=1}
         <at 0.1 0>
         {arriveTangent=1.1111, interpMode=RCIM_Cubic, leaveTangent=1.1111, time=0.4, value=0}
         {interpMode=RCIM_Cubic, leaveTangent=16.4602, time=1, value=1}
        ]}}})
    Index 0 Param 2=&FloatFromCurve(
     FloatCurve=DataInterface() {propertyValues {Curve {keys [
         {arriveTangent=2.4843, interpMode=RCIM_Cubic, leaveTangent=2.4843, tangentMode=RCTM_User, time=0, value=0}
         {interpMode=RCIM_Cubic, leaveTangent=-0.8461, time=1, value=1}
        ]}}}
     Scale Curve=0.5
    )
    Index 0 Param 4=&FloatFromCurve(
     FloatCurve=DataInterface() {propertyValues {Curve {keys [
         {arriveTangent=0.2805, interpMode=RCIM_Cubic, time=0, value=1}
         {arriveTangent=2.4494, interpMode=RCIM_Cubic, leaveTangent=2.4494, tangentMode=RCTM_User, time=1, value=2}
        ]}}}
     Scale Curve=&RandomRangeFloat(Minimum=0.6, Maximum=0.8)
    )
    Module.Param0WriteEnabled=true
    Module.Param1WriteEnabled=true
    Module.Param2WriteEnabled=false
    Module.Param3WriteEnabled=true
   )
  ]
  rendererClasses [&NiagaraMeshRendererProperties]
  renderers [{rendererClass=&NiagaraMeshRendererProperties, rendererIndex=0}]
 }
}
summary {
 systemName=NS_ElectroBat_ImpactExplosion_Heavy
 emitters [
  {emitterName=Lit, rendererClasses [&NiagaraLightRendererProperties]}
  {emitterName=LW_Lightning, rendererClasses []}
  {emitterName=LW_Flash, rendererClasses []}
  {emitterName=Dark_Chunks, rendererClasses []}
  {emitterName=Empty002, rendererClasses [&NiagaraRibbonRendererProperties]}
  {emitterName=GroundWave, rendererClasses [&NiagaraMeshRendererProperties]}
  {emitterName=OuterSphere, rendererClasses [&NiagaraMeshRendererProperties, &NiagaraLightRendererProperties]}
  {emitterName=Wind, rendererClasses [&NiagaraMeshRendererProperties]}
 ]
 userVariables [
  User.ChargeCount {type=NiagaraFloat, defaultValue=0, description=""}
  User.Delay {type=NiagaraFloat, defaultValue=0, description=""}
  User.Global Scale {type=NiagaraFloat, defaultValue=1, description=""}
  User.PhysicalSurface {type=NiagaraInt32}
  User.Spawn Stage2 {type=NiagaraFloat, defaultValue=1, description=""}
  User.Spawn Stage3 {type=NiagaraFloat, defaultValue=1, description=""}
  User.Spawn Stage4 {type=NiagaraFloat, defaultValue=1, description=""}
  User.WaveDirection {type=Vector3f, description="", defaultValue {struct=Vector3f, value=<1 0 0>}}
 ]
}
