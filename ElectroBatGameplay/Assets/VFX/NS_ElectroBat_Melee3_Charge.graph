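# Niagara system for the charge effect of ElectroBat's third melee attack, built from sprites attached to skeletal meshes.
# Decals bursts nine cyan sprites, partitioned across body mesh bones or sockets, with material parameters driven by system age and normalized velocity.
# Fountain places short-lived sprites on a second skeletal mesh, pulsing their size with a fast sine and brightening from blue toward cyan.
# Both use camera offsets and color scaling to intensify glow.
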NiagaraSystem /ElectroBatGameplay/Assets/VFX/NS_ElectroBat_Melee3_Charge
dependencies {
 usedDataInterfaces [&NiagaraDataInterfaceCurve, &NiagaraDataInterfaceColorCurve, &NiagaraDataInterfaceSkeletalMesh]
 usedDynamicInputs [&Multiply_VectorByFloat]
 usedModules [
  &self:SystemSpawnScript.SystemScriptSource.SystemScriptGraph.NiagaraNodeAssignment_0.SetVariables_#1
  &self:SystemSpawnScript.SystemScriptSource.SystemScriptGraph.NiagaraNodeAssignment_1.SetVariables_#2
  &SystemState
 ]
}
emitters {
 Decals {
  emitterUpdateScript [
   &EmitterState()
   &SpawnBurst_Instantaneous(Spawn Count=9, Use Spawn Probability=false, UseLoopCountLimit=false)
  ]
  particleSpawnScript [
   &PartitionParticles(Number Of Partitions=9)
   &InitializeParticle(
    Color=<rgba 0 0.9298 3 1>
    Lifetime=0.5
    Sprite Size Mode=::NewEnumerator3
    Uniform Sprite Size=30
    UsePositionOffset=false
    Write Material Random=true
   )
   &CameraOffset(Camera Offset Amount=10)
  ]
  particleUpdateScript [
   &ParticleState()
   &SkeletalMeshLocation(
    Bone / Socket Index=$Particles.PartitionParticles.Partition
    Bone Sampling Mode=::NewEnumerator4
    MaskSpawnGroup=false
    Orientation Sampling=::NewEnumerator1
    Skeletal Mesh=$System.SkeletalMesh_Body
   )
   /Niagara/Modules/Update/Utility/IncrementOverTime(Rate of Change:NiagaraFloat=&Lerp_Float(A=-1.1, B=2, Alpha=$System.NormalizedVelocityRange))
   &DynamicMaterialParameters(
    Index 0 Param 1=&FloatFromCurve(
     FloatCurve=DataInterface() {propertyValues {Curve {keys [{arriveTangent=3.2804, interpMode=RCIM_Cubic, leaveTangent=3.2804, tangentMode=RCTM_User, time=0, value=1}, <at 0.7 2>]}}}
     CurveIndex=$System.Age
    )
    Index 0 Param 2=1
    Index 0 Param 3=&Add_Float(
     A=$Particles.IncrementOverTime.Counter
     B=&FloatFromCurve(
      FloatCurve=DataInterface() {propertyValues {Curve {keys [{arriveTangent=-2.1645, interpMode=RCIM_Cubic, leaveTangent=-2.1645, tangentMode=RCTM_User, time=0, value=1}, <at 1.25 0>]}}}
      CurveIndex=$System.Age
      Scale Curve=-0.5
     )
    )
    Index 0 Param 4=&RandomRangeFloat(Minimum=0.75, Maximum=0.9)
    Index 1 Param 1=50
    Index 1 Param 2=10
    Module.Param0WriteEnabled=true
    Module.Param0WriteEnabled1=true
    Module.Param1WriteEnabled=true
    Module.Param1WriteEnabled1=true
    Module.Param2WriteEnabled=true
    Module.Param2WriteEnabled1=false
    Module.Param3WriteEnabled=true
    Module.Param3WriteEnabled1=false
    Write Parameter Index 1=true
   )
   &ScaleColor(
    Scale RGB=&VectorFromFloat(Value=&Lerp_Float(
      A=1
      B=6
      Alpha=&FloatFromCurve(FloatCurve=DataInterface() {propertyValues {Curve {keys [
           {interpMode=RCIM_Cubic, tangentMode=RCTM_User, time=0, value=0}
           {arriveTangent=2.1597, interpMode=RCIM_Cubic, leaveTangent=2.1597, tangentMode=RCTM_User, time=0.75, value=1}
           {tangentMode=RCTM_User, time=1, value=0}
          ]}}})
     ))
    ScaleA=true
    ScaleRGB=true
   )
  ]
  rendererClasses [&NiagaraSpriteRendererProperties]
  renderers [{rendererClass=&NiagaraSpriteRendererProperties}]
 }
 Fountain {
  emitterUpdateScript [
   &EmitterState(Life Cycle Mode=::NewEnumerator1, Loop Behavior=::NewEnumerator1, Loop Duration=0.25, UseLoopDelay=false)
   &SpawnBurst_Instantaneous(Use Spawn Probability=false, UseLoopCountLimit=false)
   &SpawnRate(SpawnRate=90, Use Spawn Probability=false) {enabled=false}
  ]
  particleSpawnScript [
   &InitializeParticle(
    Color=<rgba 1 1 1 0.5>
    Lifetime Max=0.4
    Lifetime Min=0.4
    Lifetime Mode=::NewEnumerator1
    Mass Mode=::NewEnumerator3
    Sprite Rotation Mode=::NewEnumerator2
    Sprite Size Mode=::NewEnumerator1
    Uniform Sprite Size Max=170
    Uniform Sprite Size Min=150
    UsePositionOffset=false
    Write Material Random=true
   )
   &SkeletalMeshLocation(
    Bone Sampling Mode=::NewEnumerator4
    MaskSpawnGroup=false
    Orientation Sampling=::NewEnumerator1
    Sampled Position Offset=<1 0 0>
    Skeletal Mesh=$System.SkeletalMesh
   )
  ]
  particleUpdateScript [
   &ParticleState()
   &ScaleSpriteSize(
    Uniform Curve Scale=&Sine(Normalized Angle=$Emitter.Age, Period=0.1, Scale=0.1, Bias=0.9)
    Uniform Curve Sprite Scale=DataInterface() {propertyValues {Curve {keys [
        {tangentMode=RCTM_User, time=0, value=0.1}
        {arriveTangent=-3.4334, interpMode=RCIM_Cubic, leaveTangent=-3.4334, tangentMode=RCTM_User, time=0.6, value=0.55}
        {tangentMode=RCTM_User, time=1, value=0.025}
       ]}}}
   )
   &ScaleColor(
    Linear Color Curve=DataInterface() {propertyValues {
      AlphaCurve {keys [<at 0 0.1>,<at 0.9527 1>,<at 1 0>]}
      BlueCurve {keys [<at 0.5338 15>,<at 0.7937 50>]}
      GreenCurve {keys [<at 0.5338 1.5467>,<at 0.7937 34.2627>]}
      RedCurve {keys [<at 0.5338 0>,<at 0.7937 0>]}
     }}
    Scale Mode=::NewEnumerator2
   )
   &CameraOffset(Camera Offset Amount=&FloatFromCurve(
     FloatCurve=DataInterface() {propertyValues {Curve {keys [{tangentMode=RCTM_User, time=0, value=0}, {tangentMode=RCTM_User, time=0.68, value=1}]}}}
     Scale Curve=20
    ))
  ]
  rendererClasses [&NiagaraSpriteRendererProperties]
  renderers [
   {rendererClass=&NiagaraSpriteRendererProperties}
   {rendererClass=&NiagaraSpriteRendererProperties, rendererIndex=1}
  ]
 }
}
summary {
 systemName=NS_ElectroBat_Melee3_Charge
 emitters [
  {emitterName=Decals, rendererClasses [&NiagaraSpriteRendererProperties]}
  {emitterName=Fountain, rendererClasses [&NiagaraSpriteRendererProperties]}
 ]
 userVariables []
}
