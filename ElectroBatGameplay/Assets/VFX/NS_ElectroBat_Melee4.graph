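# Niagara system for the fourth ElectroBat melee effect, named as an electric hit with ground impact elements.
# Only the debris and ground smoke emitters are enabled; sparks, lightning, streaks and ground emitters are present but disabled.
# User parameters expose delay, global scale, a cyan-blue HDR color, wave direction and spawn values for stages 2 to 4.
# The system script uses SystemState plus a variable-assignment module; emitter internals and renderers are not listed.
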NiagaraSystem /ElectroBatGameplay/Assets/VFX/NS_ElectroBat_Melee4
dependencies {
 usedDataInterfaces []
 usedDynamicInputs []
 usedModules [
  &self:SystemSpawnScript.SystemScriptSource.SystemScriptGraph.NiagaraNodeAssignment_0.SetVariables_#1
  &SystemState
 ]
}
emitters {
 LW_Debris {rendererClasses [], renderers []}
 LW_Ground001 {bEnabled=false, rendererClasses [], renderers []}
 LW_Lightning {bEnabled=false, rendererClasses [], renderers []}
 LW_Smoke_Ground {rendererClasses [], renderers []}
 LW_Sparks1 {bEnabled=false, rendererClasses [], renderers []}
 LW_Streaks {bEnabled=false, rendererClasses [], renderers []}
}
summary {
 systemName=NS_ElectroBat_Melee4
 emitters [
  {bEnabled=false, emitterName=LW_Sparks1, rendererClasses []}
  {emitterName=LW_Debris, rendererClasses []}
  {bEnabled=false, emitterName=LW_Lightning, rendererClasses []}
  {bEnabled=false, emitterName=LW_Streaks, rendererClasses []}
  {bEnabled=false, emitterName=LW_Ground001, rendererClasses []}
  {emitterName=LW_Smoke_Ground, rendererClasses []}
 ]
 userVariables [
  User.Delay {type=NiagaraFloat, defaultValue=0, description=""}
  User.Global Color {type=LinearColor, description="", defaultValue {struct=LinearColor, value=<rgba 0.0081 1.6445 4 1>}}
  User.Global Scale {type=NiagaraFloat, defaultValue=1, description=""}
  User.Spawn Stage2 {type=NiagaraFloat, defaultValue=1, description=""}
  User.Spawn Stage3 {type=NiagaraFloat, defaultValue=1, description=""}
  User.Spawn Stage4 {type=NiagaraFloat, defaultValue=1, description=""}
  User.WaveDirection {type=Vector3f, description="", defaultValue {struct=Vector3f, value=<1 0 0>}}
 ]
}
