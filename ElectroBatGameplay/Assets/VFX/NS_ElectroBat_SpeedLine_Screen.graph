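# Niagara screen speed-line effect for the ElectroBat, drawing short-lived blue sprite streaks.
# Twin CPU and GPU emitters share the same stack, with the GPU one spawning far more densely.
# Sprites spawn on a stretched ring shape with added velocity, stretch by speed, fade over life and drive dynamic material parameters.
# A RegionMask lerps sprite size to zero, hiding streaks by spawn region.
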NiagaraSystem /ElectroBatGameplay/Assets/VFX/NS_ElectroBat_SpeedLine_Screen
dependencies {usedDataInterfaces [&NiagaraDataInterfaceCurve], usedDynamicInputs [], usedModules [&SystemState]}
emitters {
 Screen_SpeedLine_CPU {
  emitterUpdateScript [
   &EmitterState()
   &SpawnRate(SpawnRate=80, Use Spawn Probability=false)
  ]
  particleSpawnScript [
   &InitializeParticle(
    Color Maximum=&MultiplyLinearColor(Linear Color=<rgba 0.2948 0.3073 0.1369 1>, Scale Factor=<1 1 1 1>)
    Color Minimum=<rgba 0.1454 0.2917 0.1241 1>
    Color Mode=::NewEnumerator2
    Lifetime Max=1
    Lifetime Min=0.6
    Lifetime Mode=::NewEnumerator1
    Sprite Size=&RandomRangeVector2D(Minimum=<0.84 0.7>, Maximum=<0.98 1.26>)
    Sprite Size Mode=::NewEnumerator4
    UsePositionOffset=false
    Write Material Random=true
   ) {enabled=false}
   /Niagara/Modules/Spawn/Initialization/InitializeParticle(
    Color=<rgba 0 0.3843 1 1>
    Lifetime=&UniformRangedFloat(Minimum=0.1, Maximum=0.2)
    Module.Write Color=true
    Module.Write Lifetime=true
    Module.Write Mass=false
    Module.Write Material Random=true
    Module.Write Position=false
    Module.Write Scale=false
    Module.Write SpriteRotation=false
    Module.Write SpriteSize=true
    Position=&SimulationPosition()
    Sprite Size=&MaskVector2dBySpawnGroup(
     Masked Vector2D=&UniformRangedVector2D(Minimum=<0.4 1>, Maximum=<0.7 1>)
     Passthrough Vector2D=&UniformRangedVector2D(Minimum=<0.8 1>, Maximum=<1.2 1>)
     Spawn Group Mask=0
    )
   ) {module=InitializeParticle001}
   &ShapeLocation(
    MaskSpawnGroup=false
    Non Uniform Scale=<1 1.7 1>
    Override Seed=false
    Ring Radius=45
    Rotation Mode=::NewEnumerator1
    Shape Primitive=::NewEnumerator4
    UseOwnerScale=true
   )
   &RegionMask(
    Region Dimensions:Vector3f=&Multiply_VectorByFloat(Vector=<100 250 150>, Float=0.6)
    Region Falloff=0
    Region Type:ENiagaraDebugDrawShape=::NewEnumerator1
    UseOffset=false
    UseRotation=false
   )
   &AddVelocity(Velocity=&RandomRangeVector(Minimum=<0 0 0>, Maximum=<0 0 -3>))
   &AddVelocity(Module.Use Falloff Distance=false, Override Seed=false, Velocity Mode=::NewEnumerator1, Velocity Speed=600) {module=AddVelocity001}
  ]
  particleUpdateScript [
   &ParticleState()
   &ScaleSpriteSize(Scale Sprite Size Mode=::NewEnumerator0, Uniform Scale Factor=2)
   &ScaleSpriteSize(
    Scale Sprite Size Mode=::NewEnumerator0
    Uniform Scale Factor=&Lerp_Float(A=1, B=0, Alpha=$Particles.RegionMask.Float)
   ) {module=ScaleSpriteSize001}
   &ScaleColor(
    Scale Alpha=&FloatFromCurve(FloatCurve=DataInterface() {propertyValues {Curve {keys [
         {arriveTangent=6.3626, interpMode=RCIM_Cubic, time=0, value=0}
         <at 0.5 0.9779 RCIM_Cubic>
         {interpMode=RCIM_Cubic, leaveTangent=-0.6839, time=1, value=0}
        ]}}})
    ScaleA=true
    ScaleRGB=false
   )
   &SolveForcesAndVelocity(Clamp Velocity=false, Limit Acceleration=false)
   &ScaleSpriteSizeBySpeed(Max Scale Factor=<1 20>, Min Scale Factor=<0.5 15>)
   &DynamicMaterialParameters(
    Index 0 Param 1=&FloatFromCurve(FloatCurve=DataInterface() {propertyValues {Curve {keys [
         <at 0 1 RCIM_Cubic>
         {arriveTangent=-0.0571, interpMode=RCIM_Cubic, leaveTangent=-0.0571, time=0.012, value=0.95}
         {arriveTangent=0.05, leaveTangent=0.05, time=0.8761, value=0.95}
         <at 1 1>
        ]}}})
    Index 0 Param 2=1
    Index 0 Param 3=1
    Index 0 Param 4=1
    Index 1 Param 1=&RandomRangeFloat(Minimum=0.5, Maximum=1)
    Index 1 Param 2=&RandomRangeFloat(Minimum=1, Maximum=2)
    Index 1 Param 3=&FloatFromCurve(FloatCurve=DataInterface() {propertyValues {Curve {keys [
         {tangentMode=RCTM_User, time=0, value=1}
         {arriveTangent=-0.0405, interpMode=RCIM_Cubic, leaveTangent=-0.0405, tangentMode=RCTM_User, time=0.7009, value=1}
         {arriveTangent=-8.0089, interpMode=RCIM_Cubic, leaveTangent=-8.0089, tangentMode=RCTM_User, time=1, value=0}
        ]}}})
    Module.Param0WriteEnabled=true
    Module.Param0WriteEnabled1=true
    Module.Param1WriteEnabled=true
    Module.Param1WriteEnabled1=true
    Module.Param2WriteEnabled=true
    Module.Param2WriteEnabled1=true
    Module.Param3WriteEnabled=true
    Module.Param3WriteEnabled1=true
    Write Parameter Index 1=true
   )
  ]
  rendererClasses [&NiagaraSpriteRendererProperties]
  renderers [{rendererClass=&NiagaraSpriteRendererProperties}]
 }
 Screen_SpeedLine_GPU {
  simTarget=GPUComputeSim
  emitterUpdateScript [
   &EmitterState()
   &SpawnRate(SpawnRate=300, Use Spawn Probability=false)
  ]
  particleSpawnScript [
   /Niagara/Modules/Spawn/Initialization/V2/InitializeParticle(
    Color Maximum=&MultiplyLinearColor(Linear Color=<rgba 0.2948 0.3073 0.1369 1>, Scale Factor=<1 1 1 1>)
    Color Minimum=<rgba 0.1454 0.2917 0.1241 1>
    Color Mode=::NewEnumerator2
    Lifetime Max=1
    Lifetime Min=0.6
    Lifetime Mode=::NewEnumerator1
    Sprite Size=&RandomRangeVector2D(Minimum=<0.84 0.7>, Maximum=<0.98 1.26>)
    Sprite Size Mode=::NewEnumerator4
    UsePositionOffset=false
    Write Material Random=true
   ) {enabled=false}
   /Niagara/Modules/Spawn/Initialization/InitializeParticle(
    Color=<rgba 0 0.3843 1 1>
    Lifetime=&UniformRangedFloat(Minimum=0.1, Maximum=0.2)
    Module.Write Color=true
    Module.Write Lifetime=true
    Module.Write Mass=false
    Module.Write Material Random=true
    Module.Write Position=false
    Module.Write Scale=false
    Module.Write SpriteRotation=false
    Module.Write SpriteSize=true
    Position=&SimulationPosition()
    Sprite Size=&MaskVector2dBySpawnGroup(
     Masked Vector2D=&UniformRangedVector2D(Minimum=<0.4 1>, Maximum=<0.7 1>)
     Passthrough Vector2D=&UniformRangedVector2D(Minimum=<0.8 1>, Maximum=<1.2 1>)
     Spawn Group Mask=0
    )
   ) {module=InitializeParticle001}
   &ShapeLocation(
    MaskSpawnGroup=false
    Non Uniform Scale=<1 1.7 1>
    Override Seed=false
    Ring Radius=45
    Rotation Mode=::NewEnumerator1
    Shape Primitive=::NewEnumerator4
    UseOwnerScale=true
   )
   &RegionMask(
    Region Dimensions:Vector3f=&Multiply_VectorByFloat(Vector=<100 250 150>, Float=0.6)
    Region Falloff=0
    Region Type:ENiagaraDebugDrawShape=::NewEnumerator1
    UseOffset=false
    UseRotation=false
   )
   &AddVelocity(Velocity=&RandomRangeVector(Minimum=<0 0 0>, Maximum=<0 0 -3>))
   &AddVelocity(Module.Use Falloff Distance=false, Override Seed=false, Velocity Mode=::NewEnumerator1, Velocity Speed=600) {module=AddVelocity001}
  ]
  particleUpdateScript [
   &ParticleState()
   &ScaleSpriteSize(Scale Sprite Size Mode=::NewEnumerator0, Uniform Scale Factor=2)
   &ScaleSpriteSize(
    Scale Sprite Size Mode=::NewEnumerator0
    Uniform Scale Factor=&Lerp_Float(A=1, B=0, Alpha=$Particles.RegionMask.Float)
   ) {module=ScaleSpriteSize001}
   &ScaleColor(
    Scale Alpha=&FloatFromCurve(FloatCurve=DataInterface() {propertyValues {Curve {keys [
         {arriveTangent=6.3626, interpMode=RCIM_Cubic, time=0, value=0}
         <at 0.5 0.9779 RCIM_Cubic>
         {interpMode=RCIM_Cubic, leaveTangent=-0.6839, time=1, value=0}
        ]}}})
    ScaleA=true
    ScaleRGB=false
   )
   &SolveForcesAndVelocity(Clamp Velocity=false, Limit Acceleration=false)
   &ScaleSpriteSizeBySpeed(Max Scale Factor=<1 20>, Min Scale Factor=<0.5 15>)
   &DynamicMaterialParameters(
    Index 0 Param 1=&FloatFromCurve(FloatCurve=DataInterface() {propertyValues {Curve {keys [
         <at 0 1 RCIM_Cubic>
         {arriveTangent=-0.0571, interpMode=RCIM_Cubic, leaveTangent=-0.0571, time=0.012, value=0.95}
         {arriveTangent=0.05, leaveTangent=0.05, time=0.8761, value=0.95}
         <at 1 1>
        ]}}})
    Index 0 Param 2=1
    Index 0 Param 3=1
    Index 0 Param 4=1
    Index 1 Param 1=&RandomRangeFloat(Minimum=0.5, Maximum=1)
    Index 1 Param 2=&RandomRangeFloat(Minimum=1, Maximum=2)
    Index 1 Param 3=&FloatFromCurve(FloatCurve=DataInterface() {propertyValues {Curve {keys [
         {tangentMode=RCTM_User, time=0, value=1}
         {arriveTangent=-0.0405, interpMode=RCIM_Cubic, leaveTangent=-0.0405, tangentMode=RCTM_User, time=0.7009, value=1}
         {arriveTangent=-8.0089, interpMode=RCIM_Cubic, leaveTangent=-8.0089, tangentMode=RCTM_User, time=1, value=0}
        ]}}})
    Module.Param0WriteEnabled=true
    Module.Param0WriteEnabled1=true
    Module.Param1WriteEnabled=true
    Module.Param1WriteEnabled1=true
    Module.Param2WriteEnabled=true
    Module.Param2WriteEnabled1=true
    Module.Param3WriteEnabled=true
    Module.Param3WriteEnabled1=true
    Write Parameter Index 1=true
   )
  ]
  rendererClasses [&NiagaraSpriteRendererProperties]
  renderers [{rendererClass=&NiagaraSpriteRendererProperties}]
 }
}
summary {
 systemName=NS_ElectroBat_SpeedLine_Screen
 emitters [
  {emitterName=Screen_SpeedLine_GPU, simTarget=GPUComputeSim, rendererClasses [&NiagaraSpriteRendererProperties]}
  {emitterName=Screen_SpeedLine_CPU, rendererClasses [&NiagaraSpriteRendererProperties]}
 ]
 userVariables [
  User.ColorVariant {type=NiagaraInt32}
  User.DelayTime {type=NiagaraFloat, defaultValue=0, description=""}
  User.FlashIntensity {type=NiagaraFloat, defaultValue=0.7, description=""}
  User.IsLocalPlayer {type=NiagaraBool, defaultValue=true}
  User.SplashEroder {type=NiagaraFloat, defaultValue=0.7, description=""}
 ]
}
