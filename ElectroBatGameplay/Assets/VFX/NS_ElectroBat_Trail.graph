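# Niagara anim trail for the ElectroBat, drawing electric ribbon streaks along skeletal mesh bones.
# Two active ribbon emitters use FN_AnimTrail modules: a short cyan-based streak whose width scales with User.Super, and a faint bright layer whose alpha cuts off early in emitter age.
# Curves over particle life shape color, alpha, width and dynamic material parameters.
# Two other ribbon variants and the Lightning_Strokes sprite emitter are disabled.
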NiagaraSystem /ElectroBatGameplay/Assets/VFX/NS_ElectroBat_Trail
dependencies {
 usedDataInterfaces [
  &NiagaraDataInterfaceCurve
  &NiagaraDataInterfaceColorCurve
  &NiagaraDataInterfaceSpriteRendererInfo
 ]
 usedDynamicInputs []
 usedModules [
  &self:SystemSpawnScript.SystemScriptSource.SystemScriptGraph.NiagaraNodeAssignment_0.SetVariables_#1
  &SystemState
 ]
}
emitters {
 FN_AnimTrail {
  emitterSpawnScript [&FN_RibbonIDOffsetUtility()]
  emitterUpdateScript [
   &EmitterState()
   &FN_SpawnAnimTrail()
  ]
  particleSpawnScript [
   &FN_InitializeAnimTrailBoneIndecies()
   &FN_InitializeAnimTrail(
    AnimTrailWidth=&Multiply_Float(A=3, B=$User.Super)
    Color=&MakeLinearColorFromVectorAndFloat(
     "Vector (RGB)"=&MakeVectorFromLinearColorRGB(Color=<rgba 1.9979 13.1308 15 0.2609>)
     "Float (Alpha)"=&FloatFromCurve(FloatCurve=DataInterface() {propertyValues {Curve {keys [<at 0 0>,<at 0.1 2>,<at 1 1>]}}}, CurveIndex=$Emitter.NormalizedLoopAge)
    )
    Lifetime=0.14
    Position Offset=<0 -10 0>
    SkeletalMesh=$System.Skeletal Mesh
   )
  ]
  particleUpdateScript [
   &SolveForcesAndVelocity(Clamp Velocity=false, Limit Acceleration=false)
   &FN_UpdateAnimTrail(Locked Particle RibbonWidth=1, SkeletalMesh=$System.Skeletal Mesh) {module=FN_UpdateAnimTrail001}
   &ScaleColor(
    Scale Alpha=&FloatFromCurve(
     FloatCurve=DataInterface() {propertyValues {Curve {keys [
         {interpMode=RCIM_Cubic, leaveTangentWeight=0.0492, time=0, value=0}
         {arriveTangentWeight=0.0544, interpMode=RCIM_Cubic, leaveTangentWeight=0.1448, tangentMode=RCTM_User, time=0.1476, value=0.9816}
         {
          arriveTangent=-1.6758
          arriveTangentWeight=0.1997
          interpMode=RCIM_Cubic
          leaveTangent=-1.6758
          leaveTangentWeight=0.2336
          tangentMode=RCTM_User
          time=0.5405
          value=0.2549
         }
         {arriveTangentWeight=0.1532, interpMode=RCIM_Cubic, time=1, value=0}
        ]}}}
     Scale Curve=0.4
    )
    Scale RGB=&MakeVectorFromLinearColorRGB(Color=&ColorFromCurve(ColorCurve=DataInterface() {propertyValues {
        AlphaCurve {keys []}
        BlueCurve {keys [<at 0 1 RCIM_Cubic>,<at 0.1 1>,<at 0.25 0.2>,<at 1 0.1 RCIM_Cubic>]}
        GreenCurve {keys [<at 0 0.2746 RCIM_Cubic>,<at 0.1 0.1612>,<at 0.25 0.1858>,<at 1 0.1 RCIM_Cubic>]}
        RedCurve {keys [<at 0 0.8916 RCIM_Cubic>,<at 0.1 0.7817>,<at 0.25 0.1578>,<at 1 0.1 RCIM_Cubic>]}
       }}))
    ScaleA=true
    ScaleRGB=true
   )
   &DynamicMaterialParameters(
    Index 0 Param 1=&FloatFromCurve(FloatCurve=DataInterface() {propertyValues {Curve {keys [<at 0 1 RCIM_Cubic>, {arriveTangent=-1, interpMode=RCIM_Cubic, leaveTangent=-1, time=0.2695, value=0.2392}, <at 1 0 RCIM_Cubic>]}}})
    Index 0 Param 2=&FloatFromCurve(FloatCurve=DataInterface() {propertyValues {Curve {keys [<at 0 1 RCIM_Cubic>,<at 0.5 0>]}}})
    Index 0 Param 3=-1
    Index 0 Param 4=1
    Module.Param0WriteEnabled=true
    Module.Param1WriteEnabled=true
    Module.Param2WriteEnabled=true
    Module.Param3WriteEnabled=true
   )
   &ScaleRibbonWidth(Ribbon Width Scale=&FloatFromCurve(FloatCurve=DataInterface() {propertyValues {Curve {keys [
         {arriveTangent=15.7099, interpMode=RCIM_Cubic, leaveTangent=15.7099, tangentMode=RCTM_User, time=0, value=0}
         {arriveTangent=0.4861, interpMode=RCIM_Cubic, leaveTangent=0.4861, tangentMode=RCTM_User, time=0.2, value=0.95}
         {arriveTangent=-1.1875, interpMode=RCIM_Cubic, leaveTangent=-1.1875, time=0.5, value=0.625}
         <at 1 0 RCIM_Cubic>
        ]}}}))
  ]
  rendererClasses [&NiagaraRibbonRendererProperties]
  renderers [{rendererClass=&NiagaraRibbonRendererProperties}]
 }
 FN_AnimTrail001 {
  bEnabled=false
  emitterSpawnScript [&FN_RibbonIDOffsetUtility()]
  emitterUpdateScript [
   &EmitterState()
   &FN_SpawnAnimTrail()
  ]
  particleSpawnScript [
   &FN_InitializeAnimTrailBoneIndecies()
   &FN_InitializeAnimTrail(
    AnimTrailWidth=&Multiply_Float(A=1.5, B=$User.Super)
    Color=&MakeLinearColorFromVectorAndFloat(
     "Vector (RGB)"=&MakeVectorFromLinearColorRGB(Color=<rgba 0 0.513 0.4856 1>)
     "Float (Alpha)"=&FloatFromCurve(FloatCurve=DataInterface() {propertyValues {Curve {keys [<at 0 0>,<at 0.1 2>,<at 1 1>]}}}, CurveIndex=$Emitter.NormalizedLoopAge)
    )
    Lifetime=0.1
    Position Offset=<0 -10 0>
    SkeletalMesh=$System.Skeletal Mesh
   )
  ]
  particleUpdateScript [
   &SolveForcesAndVelocity(Clamp Velocity=false, Limit Acceleration=false)
   &FN_UpdateAnimTrail(Locked Particle RibbonWidth=1, SkeletalMesh=$System.Skeletal Mesh) {module=FN_UpdateAnimTrail001}
   &ScaleColor(
    Scale Alpha=&FloatFromCurve(FloatCurve=DataInterface() {propertyValues {Curve {keys [
         {interpMode=RCIM_Cubic, leaveTangentWeight=0.0492, time=0, value=0}
         {arriveTangentWeight=0.0544, interpMode=RCIM_Cubic, leaveTangentWeight=0.1448, tangentMode=RCTM_User, time=0.1476, value=0.9816}
         {
          arriveTangent=-1.6758
          arriveTangentWeight=0.1997
          interpMode=RCIM_Cubic
          leaveTangent=-1.6758
          leaveTangentWeight=0.2336
          tangentMode=RCTM_User
          time=0.5405
          value=0.2549
         }
         {arriveTangentWeight=0.1532, interpMode=RCIM_Cubic, time=1, value=0}
        ]}}})
    Scale RGB=&MakeVectorFromLinearColorRGB(Color=&ColorFromCurve(ColorCurve=DataInterface() {propertyValues {
        AlphaCurve {keys []}
        BlueCurve {keys [<at 0 1 RCIM_Cubic>,<at 0.1 1>,<at 0.25 0.2>,<at 1 0.1 RCIM_Cubic>]}
        GreenCurve {keys [<at 0 0.2746 RCIM_Cubic>,<at 0.1 0.1612>,<at 0.25 0.1858>,<at 1 0.1 RCIM_Cubic>]}
        RedCurve {keys [<at 0 0.8916 RCIM_Cubic>,<at 0.1 0.7817>,<at 0.25 0.1578>,<at 1 0.1 RCIM_Cubic>]}
       }}))
    ScaleA=true
    ScaleRGB=true
   )
   &DynamicMaterialParameters(
    Index 0 Param 1=&FloatFromCurve(FloatCurve=DataInterface() {propertyValues {Curve {keys [<at 0 1 RCIM_Cubic>, {arriveTangent=-1, interpMode=RCIM_Cubic, leaveTangent=-1, time=0.2695, value=0.2392}, <at 1 0 RCIM_Cubic>]}}})
    Index 0 Param 2=&FloatFromCurve(FloatCurve=DataInterface() {propertyValues {Curve {keys [<at 0 1 RCIM_Cubic>,<at 0.5 0>]}}})
    Index 0 Param 3=-1
    Index 0 Param 4=1
    Module.Param0WriteEnabled=true
    Module.Param1WriteEnabled=true
    Module.Param2WriteEnabled=true
    Module.Param3WriteEnabled=true
   )
   &ScaleRibbonWidth(Ribbon Width Scale=&FloatFromCurve(FloatCurve=DataInterface() {propertyValues {Curve {keys [
         {arriveTangent=15.7099, interpMode=RCIM_Cubic, leaveTangent=15.7099, tangentMode=RCTM_User, time=0, value=0}
         {arriveTangent=0.4861, interpMode=RCIM_Cubic, leaveTangent=0.4861, tangentMode=RCTM_User, time=0.2, value=0.95}
         {arriveTangent=-1.1875, interpMode=RCIM_Cubic, leaveTangent=-1.1875, time=0.5, value=0.625}
         <at 1 0 RCIM_Cubic>
        ]}}}))
  ]
  rendererClasses [&NiagaraRibbonRendererProperties]
  renderers [{rendererClass=&NiagaraRibbonRendererProperties}]
 }
 FN_AnimTrail002 {
  bEnabled=false
  emitterSpawnScript [&FN_RibbonIDOffsetUtility()]
  emitterUpdateScript [
   &EmitterState(Life Cycle Mode=::NewEnumerator1, Loop Behavior=::NewEnumerator1, Loop Duration=0.1, UseLoopDelay=false)
   &FN_SpawnAnimTrail()
  ]
  particleSpawnScript [
   &FN_InitializeAnimTrailBoneIndecies()
   &FN_InitializeAnimTrail(
    Color=&MakeLinearColorFromVectorAndFloat(
     "Vector (RGB)"=&MakeVectorFromLinearColorRGB(Color=<rgba 100 100 100 1>)
     "Float (Alpha)"=&FloatFromCurve(FloatCurve=DataInterface() {propertyValues {Curve {keys [<at 0 0>,<at 0.1 2>,<at 1 1>]}}}, CurveIndex=$Emitter.NormalizedLoopAge)
    )
    Lifetime=0.1
    Position Offset=<0 -10 0>
    SkeletalMesh=$System.Skeletal Mesh
   )
  ]
  particleUpdateScript [
   &SolveForcesAndVelocity(Clamp Velocity=false, Limit Acceleration=false)
   &FN_UpdateAnimTrail(Locked Particle RibbonWidth=1, SkeletalMesh=$System.Skeletal Mesh) {module=FN_UpdateAnimTrail001}
   &ScaleColor(
    Scale Alpha=&FloatFromCurve(FloatCurve=DataInterface() {propertyValues {Curve {keys [
         {leaveTangentWeight=0.0492, time=0, value=0.5}
         {arriveTangent=-1.1628, arriveTangentWeight=0.0544, leaveTangent=-1.1628, leaveTangentWeight=0.1448, time=0.21, value=1}
         {arriveTangentWeight=0.1997, leaveTangent=-1.6758, leaveTangentWeight=0.2336, time=0.43, value=0}
        ]}}})
    Scale RGB=&MakeVectorFromLinearColorRGB(Color=&ColorFromCurve(ColorCurve=DataInterface() {propertyValues {
        AlphaCurve {keys []}
        BlueCurve {keys [<at 0 20 RCIM_Cubic>,<at 1 0.2083 RCIM_Cubic>]}
        GreenCurve {keys [<at 0 6.9946 RCIM_Cubic>,<at 1 0 RCIM_Cubic>]}
        RedCurve {keys [<at 0 0 RCIM_Cubic>,<at 1 0.1167 RCIM_Cubic>]}
       }}))
    ScaleA=true
    ScaleRGB=true
   )
   &DynamicMaterialParameters(
    Index 0 Param 1=&FloatFromCurve(FloatCurve=DataInterface() {propertyValues {Curve {keys [<at 0 1 RCIM_Cubic>, {arriveTangent=-1, interpMode=RCIM_Cubic, leaveTangent=-1, time=0.2695, value=0.2392}, <at 1 0 RCIM_Cubic>]}}})
    Index 0 Param 2=&FloatFromCurve(FloatCurve=DataInterface() {propertyValues {Curve {keys [<at 0 1 RCIM_Cubic>,<at 0.5 0>]}}})
    Index 0 Param 3=-1
    Index 0 Param 4=1
    Module.Param0WriteEnabled=true
    Module.Param1WriteEnabled=true
    Module.Param2WriteEnabled=true
    Module.Param3WriteEnabled=true
   )
   &ScaleRibbonWidth(Ribbon Width Scale=&FloatFromCurve(FloatCurve=DataInterface() {propertyValues {Curve {keys [
         {arriveTangent=15.7099, interpMode=RCIM_Cubic, leaveTangent=15.7099, tangentMode=RCTM_User, time=0, value=0}
         {arriveTangent=0.4861, interpMode=RCIM_Cubic, leaveTangent=0.4861, tangentMode=RCTM_User, time=0.2, value=0.95}
         {arriveTangent=-1.1875, interpMode=RCIM_Cubic, leaveTangent=-1.1875, time=0.5, value=0.625}
         <at 1 0 RCIM_Cubic>
        ]}}}))
  ]
  rendererClasses [&NiagaraRibbonRendererProperties]
  renderers [{rendererClass=&NiagaraRibbonRendererProperties}]
 }
 FN_AnimTrail003 {
  emitterSpawnScript [&FN_RibbonIDOffsetUtility()]
  emitterUpdateScript [
   &EmitterState(Life Cycle Mode=::NewEnumerator1, Loop Behavior=::NewEnumerator1, Loop Duration=0.25, UseLoopDelay=false)
   &FN_SpawnAnimTrail()
  ]
  particleSpawnScript [
   &FN_InitializeAnimTrailBoneIndecies()
   &FN_InitializeAnimTrail(
    Color=&MakeLinearColorFromVectorAndFloat(
     "Vector (RGB)"=&MakeVectorFromLinearColorRGB(Color=<rgba 10 10 10 0.05>)
     "Float (Alpha)"=&FloatFromCurve(
      FloatCurve=DataInterface() {propertyValues {Curve {keys [<at 0.1 1 RCIM_Cubic>, {arriveTangent=-28.9119, tangentMode=RCTM_User, time=0.2, value=0}]}}}
      CurveIndex=$Emitter.Age
     )
    )
    Lifetime=0.2
    SkeletalMesh=$System.Skeletal Mesh
   )
  ]
  particleUpdateScript [
   &SolveForcesAndVelocity(Clamp Velocity=false, Limit Acceleration=false)
   &FN_UpdateAnimTrail(Locked Particle RibbonWidth=1, SkeletalMesh=$System.Skeletal Mesh) {module=FN_UpdateAnimTrail001}
   &ScaleColor(
    Scale Alpha=&FloatFromCurve(
     FloatCurve=DataInterface() {propertyValues {Curve {keys [
         {leaveTangentWeight=0.0492, time=0.01, value=0}
         <at 0.1 0.6>
         {arriveTangent=-1.1628, arriveTangentWeight=0.0544, leaveTangentWeight=0.1448, time=0.5, value=0.7}
         {arriveTangentWeight=0.1997, leaveTangent=-1.6758, leaveTangentWeight=0.2336, time=1, value=0}
        ]}}}
     Scale Curve=0.1
    )
    Scale RGB=&MakeVectorFromLinearColorRGB(Color=&ColorFromCurve(ColorCurve=DataInterface() {propertyValues {
        AlphaCurve {keys []}
        BlueCurve {keys [<at 0 20 RCIM_Cubic>,<at 1 0.2083 RCIM_Cubic>]}
        GreenCurve {keys [<at 0 6.9946 RCIM_Cubic>,<at 1 0 RCIM_Cubic>]}
        RedCurve {keys [<at 0 0 RCIM_Cubic>,<at 1 0.1167 RCIM_Cubic>]}
       }}))
    ScaleA=true
    ScaleRGB=true
   )
   &DynamicMaterialParameters(
    Index 0 Param 1=&FloatFromCurve(FloatCurve=DataInterface() {propertyValues {Curve {keys [<at 0 1 RCIM_Cubic>, {arriveTangent=-1, interpMode=RCIM_Cubic, leaveTangent=-1, time=0.2695, value=0.2392}, <at 1 0 RCIM_Cubic>]}}})
    Index 0 Param 2=&FloatFromCurve(FloatCurve=DataInterface() {propertyValues {Curve {keys [<at 0 1 RCIM_Cubic>,<at 0.5 0>]}}})
    Index 0 Param 3=-1
    Index 0 Param 4=1
    Module.Param0WriteEnabled=true
    Module.Param1WriteEnabled=true
    Module.Param2WriteEnabled=true
    Module.Param3WriteEnabled=true
   )
   &ScaleRibbonWidth(Ribbon Width Scale=&FloatFromCurve(FloatCurve=DataInterface() {propertyValues {Curve {keys [
         {arriveTangent=1.8138, interpMode=RCIM_Cubic, leaveTangent=1.8138, tangentMode=RCTM_User, time=0, value=2.75}
         {arriveTangent=-1.75, interpMode=RCIM_Cubic, leaveTangent=-1.75, time=0.3, value=1.25}
         <at 1 1 RCIM_Cubic>
        ]}}}))
  ]
  rendererClasses [&NiagaraRibbonRendererProperties]
  renderers [{rendererClass=&NiagaraRibbonRendererProperties}]
 }
 Lightning_Strokes {
  bEnabled=false
  emitterUpdateScript [
   &EmitterState()
   &SpawnPerUnit(Module.Use Max Movement Threshold=true, Module.Use Movement Tolerance=true, Movement Tolerance=0.1, Use Spawn Probability=false)
   &SpawnRate(SpawnRate=72, Use Spawn Probability=false) {enabled=false}
  ]
  particleSpawnScript [
   &InitializeParticle(
    Color=<rgba 0.0372 1.8936 2 1>
    Lifetime=0.1
    Sprite Rotation Mode=::NewEnumerator2
    Sprite Size Mode=::NewEnumerator1
    Sprite UV Mode=::NewEnumerator3
    Uniform Sprite Size Max=40
    Uniform Sprite Size Min=20
    UsePositionOffset=false
    Write Material Random=true
   )
   &InitialMeshOrientation(Use Rotation Vector=false)
   &ShapeLocation(
    MaskSpawnGroup=false
    Offset=<0 0 70>
    Offset Coordinate Space=::Local
    Offset Mode=::NewEnumerator0
    Override Seed=false
    Sphere Radius=&RandomRangeFloat(Minimum=5, Maximum=10)
    UseOwnerScale=true
   )
  ]
  particleUpdateScript [
   &ParticleState()
   &UpdateMeshOrientation(Rotation Rate=0.2, Rotation Vector=&RandomRangeVector(Minimum=<-1 -1 -1>, Maximum=<1 1 1>))
   &DynamicMaterialParameters(
    Index 0 Param 1=&RandomRangeFloat(Minimum=2, Maximum=3)
    Index 0 Param 2=&FloatFromCurve(FloatCurve=DataInterface() {propertyValues {Curve {keys [{arriveTangent=-2.9733, interpMode=RCIM_Cubic, leaveTangent=-2.9733, tangentMode=RCTM_User, time=0, value=1}, <at 1 -0.25 RCIM_Cubic>]}}})
    Index 0 Param 3=&RandomRangeFloat(Minimum=0.2, Maximum=2)
    Index 0 Param 4=&RandomRangeFloat(Minimum=0, Maximum=1)
    Module.Param0WriteEnabled=true
    Module.Param1WriteEnabled=true
    Module.Param2WriteEnabled=true
    Module.Param3WriteEnabled=true
   )
   &ScaleColor(
    Scale RGB=&MakeVectorFromLinearColorRGB(Color=&ColorFromCurve(ColorCurve=DataInterface() {propertyValues {
        AlphaCurve {keys []}
        BlueCurve {keys [<at 0 1 RCIM_Cubic>,<at 1 0.2344 RCIM_Cubic>]}
        GreenCurve {keys [<at 0 0.8086 RCIM_Cubic>,<at 1 0.067 RCIM_Cubic>]}
        RedCurve {keys [<at 0 0.6449 RCIM_Cubic>,<at 1 0.025 RCIM_Cubic>]}
       }}))
    ScaleA=true
    ScaleRGB=true
   )
   &SubUVAnimation(
    Random Start Frame=true
    "Sprite Renderer "=DataInterface() {propertyValues {SpriteRenderer=&self:Lightning_Strokes_6.NiagaraSpriteRendererProperties_1}}
    SubUV Animation Mode:ENiagara_SubUVLookupModeV2=::NewEnumerator3
    UseEndFrame=false
    UsePlayRateScale=false
    UseStartFrame=false
   )
   &InheritSourceMovement(ApplyForce=false, ApplyPosition=true, ApplyPositionFalloff=false) {enabled=false}
   &InheritVelocity(Inherited Velocity Amount Scale=<-0.4 -0.4 0>, Inherited Velocity Speed Limit=5000, Speed Limit=true, UseSpeedThreshold=false)
   &Drag(Drag=5, Use Linear Drag=true, Use Rotational Drag=true)
   &SolveForcesAndVelocity(Clamp Velocity=false, Limit Acceleration=false)
  ]
  rendererClasses [&NiagaraSpriteRendererProperties]
  renderers [{rendererClass=&NiagaraSpriteRendererProperties}]
 }
}
summary {
 systemName=NS_ElectroBat_Trail
 emitters [
  {emitterName=FN_AnimTrail, rendererClasses [&NiagaraRibbonRendererProperties]}
  {bEnabled=false, emitterName=FN_AnimTrail001, rendererClasses [&NiagaraRibbonRendererProperties]}
  {bEnabled=false, emitterName=FN_AnimTrail002, rendererClasses [&NiagaraRibbonRendererProperties]}
  {bEnabled=false, emitterName=Lightning_Strokes, rendererClasses [&NiagaraSpriteRendererProperties]}
  {emitterName=FN_AnimTrail003, rendererClasses [&NiagaraRibbonRendererProperties]}
 ]
 userVariables [
  User.Color {type=LinearColor, description="", defaultValue {struct=LinearColor, value=<rgba 1.5601 0.8148 15 2>}}
  User.GlowColorScale {type=LinearColor, description="", defaultValue {struct=LinearColor, value=<rgba 0.7 0.02 3 2>}}
  User.RibbonIDOffset {type=NiagaraInt32}
  User.SkeletalMesh {
   type=NiagaraDataInterfaceSkeletalMesh
   defaultValue {
    struct=NiagaraExt_VariableValue_DataInterface
    value {dataInterface=&self:NiagaraDataInterfaceSkeletalMesh_0}
   }
  }
  User.Super {type=NiagaraFloat, defaultValue=1, description=""}
 ]
}
